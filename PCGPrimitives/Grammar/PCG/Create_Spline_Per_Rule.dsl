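; PCG graph that creates offset splines for grammar rule slots from input splines and slot points, carrying each rule's settings on the output.
; Slot position plus offset drives a transform of spline control points; the difference from the originals becomes the Offset Spline direction.
; Slot attributes are partly renamed and copied on, with slotScale as data-level.
; Rule assets supply an override profile grammar via Get Property, deliberately kept to retrigger the graph when assets change.
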(graph Create_Spline_Per_Rule
  (inputs (Slot%20Points :type Point) (Splines :type Spline))
  (bind DeleteAttributes_21 (Delete%20Attributes :Operation KeepSelectedAttributes :SelectedAttributes "Slot number, Slot ID, Rule ID, Is Corner, Grammar, Mirror, Pos_offset, use up vector, up vector, spawn spline meshes, Spline Mesh Angle Threshold, Set Spline Mesh Roll, Spline Roll Degrees, Rand_Pos_Range, Rand_Rot_Range, Rand_Scale_Range, Override_Scale"))
  (bind AttributeRename_0 (Attribute%20Rename :AttributeToRename Rand_Pos_Range :NewAttributeName Random_position_range))
  (bind AttributeRename_1 (Attribute%20Rename :AttributeToRename Rand_Rot_Range :NewAttributeName Random_Rotation_range))
  (bind AttributeRename_2 (Attribute%20Rename :AttributeToRename Rand_Scale_Range :NewAttributeName Random_scale_range))
  (bind AttributeRename_3 (Attribute%20Rename :AttributeToRename Override_Scale :NewAttributeName Scale_override))
  (bind PointToAttributeSet_1 (Point%20To%20Attribute%20Set))
  (bind AddAttribute_0 (Add%20Attribute :bCopyAllAttributes true :InputSource slotScale))
  (bind DeleteAttributes_3 (Delete%20Attributes :Operation KeepSelectedAttributes :SelectedAttributes "slotScale"))
  (bind PointToAttributeSet_3 (Point%20To%20Attribute%20Set))
  (bind AddAttribute_2 (Add%20Attribute :InputSource slotScale :OutputTarget @data.slotScale))
  (bind AddAttribute_19 (Add%20Attribute :InputSource Overrideprofilegrammar :OutputTarget "override profile grammar"))
  ;[0 0 0 0.5] OFFSET
  (bind AttributeMathsOp_0 (Attribute%20Maths%20Op :InputSource1 $Position :InputSource2 offset))
  (bind CopyAttributes_0 (Copy%20Attributes :InputSource $Position :OutputTarget offsetMin))
  (bind CopyAttributes_1 (Copy%20Attributes :InputSource $Position :OutputTarget offsetMax))
  (bind PointToAttributeSet_0 (Point%20To%20Attribute%20Set))
  (bind DeleteAttributes_0 (Delete%20Attributes :Operation KeepSelectedAttributes :SelectedAttributes "offsetMax, offsetMin"))
  (bind GetSplineControlPoints_0 (Get%20Spline%20Control%20Points))
  (bind TransformPoints_0 (Transform%20Points))
  (bind AttributeMathsOp_1 (Attribute%20Maths%20Op :InputSource1 $Position :InputSource2 $Position :Operation Subtract :OutputTarget Direction))
  (bind CopyAttributes_3 (Copy%20Attributes :InputSource Direction))
  (bind CreateConstant_2 (Create%20Constant :AttributeTypes (struct :Type Float :FloatValue 1.0) :OutputTarget magnitude))
  (bind OffsetSpline_0 (Offset%20Spline :DirectionAttribute Direction :MagnitudeAttribute Magnitude))
  ;[0 0 0 0.5] USING Get Property here as it forces the retrigger of the graph when the Rule assets are changed : IMPORTANT, DO NOT OPTIMIZE
  (bind DeleteAttributes_8 (Delete%20Attributes :Operation KeepSelectedAttributes :SelectedAttributes "Rule_Asset"))
  (bind PointToAttributeSet_13 (Point%20To%20Attribute%20Set))
  (bind GetPropertyFromObjectPath_0 (Get%20Property%20From%20Object%20Path :InputSource Rule_Asset :OutputAttributeName "Override profile grammar" :PropertyName Rule))
  (bind DeleteAttributes_13 (Delete%20Attributes :Operation KeepSelectedAttributes :SelectedAttributes "overrideprofilegrammar"))
  (reroute NamedRerouteDeclaration_0 Splines)
  (reroute Slot%20Point Slot%20Points)
  (connect PointToAttributeSet_1 Out AddAttribute_0 Attributes)
  (connect OffsetSpline_0 Out AddAttribute_0 In)
  (connect PointToAttributeSet_3 Out AddAttribute_2 Attributes)
  (connect AddAttribute_0 Out AddAttribute_2 In)
  (connect DeleteAttributes_13 Out AddAttribute_19 Attributes)
  (connect AddAttribute_2 Out AddAttribute_19 In)
  (connect Slot%20Point Out AttributeMathsOp_0 InA)
  (connect Slot%20Point Out AttributeMathsOp_0 InB)
  (connect TransformPoints_0 Out AttributeMathsOp_1 InA)
  (connect GetSplineControlPoints_0 Out AttributeMathsOp_1 InB)
  (connect DeleteAttributes_21 Out AttributeRename_0 In)
  (connect AttributeRename_0 Out AttributeRename_1 In)
  (connect AttributeRename_1 Out AttributeRename_2 In)
  (connect AttributeRename_2 Out AttributeRename_3 In)
  (connect AttributeMathsOp_0 Out CopyAttributes_0 Source)
  (connect AttributeMathsOp_0 Out CopyAttributes_0 Target)
  (connect CopyAttributes_0 Out CopyAttributes_1 Source)
  (connect CopyAttributes_0 Out CopyAttributes_1 Target)
  (connect AttributeMathsOp_1 Out CopyAttributes_3 Source)
  (connect NamedRerouteDeclaration_0 Out CopyAttributes_3 Target)
  (connect PointToAttributeSet_0 Out DeleteAttributes_0 In)
  (connect Slot%20Point Out DeleteAttributes_3 In)
  (connect Slot%20Point Out DeleteAttributes_8 In)
  (connect GetPropertyFromObjectPath_0 Out DeleteAttributes_13 In)
  (connect Slot%20Point Out DeleteAttributes_21 In)
  (connect PointToAttributeSet_13 Out GetPropertyFromObjectPath_0 In)
  (connect DefaultInputNode Splines GetSplineControlPoints_0 In)
  (connect CopyAttributes_3 Out OffsetSpline_0 In)
  (connect CreateConstant_2 Out OffsetSpline_0 Magnitude)
  (connect CopyAttributes_1 Out PointToAttributeSet_0 In)
  (connect AttributeRename_3 Out PointToAttributeSet_1 In)
  (connect DeleteAttributes_3 Out PointToAttributeSet_3 In)
  (connect DeleteAttributes_8 Out PointToAttributeSet_13 In)
  (connect GetSplineControlPoints_0 Out TransformPoints_0 In)
  (connect DeleteAttributes_0 Out TransformPoints_0 OffsetMax)
  (connect DeleteAttributes_0 Out TransformPoints_0 OffsetMin)
  (output Out AddAttribute_19))
