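; PCG graph that turns a Shape Grammar Definition asset into one typed attribute table of slots, rules and modules for downstream grammar operations.
; Slots are extracted and indexed, their rules flattened, numbered per slot and read from rule assets, then each rule's modules flattened.
; Defaults from a pre-saved data asset fill each level before merging and renaming.
; The commented-out branch that produced that asset remains, avoiding runtime recomputation.
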(graph Get_SGD_Data
  (params (ShapeGrammarDefinition :type (SoftObject /PCGPrimitives/Grammar/BP/BP_ShapeGrammarDefinition.BP_ShapeGrammarDefinition_C)))
  (outputs (SGD_Data :type Param))
  ;[0 0 0 0.5] Extract Slots
  (bind GetPropertyFromObjectPath_0 (Get%20Property%20From%20Object%20Path :bSanitizeOutputAttributeName false :PropertyName Slots :StructExtractorBehavior Extract :In ShapeGrammarDefinition))
  (bind AttributeSetToPoint_1 (Attribute%20Set%20To%20Point))
  (bind CopyAttributes_1 (Copy%20Attributes :InputSource $Index :OutputTarget "Slot ID"))
  (bind AddAttribute_1 (Add%20Attribute :AttributeTypes (struct :Type Integer32) :OutputTarget Type))
  ;[0 0 0 0.5] Extract Rules
  (bind MetadataArrayOperation_0 (Metadata%20Array%20Operation :InputSource1 "Shape Grammar Rules" :Operation Flatten))
  (bind AttributePartition_2 (Attribute%20Partition :PartitionAttributeSelectors ["Slot ID"]))
  (bind CopyAttributes_3 (Copy%20Attributes :InputSource $Index :OutputTarget "Rule ID"))
  (bind MergePoints_0 (Merge%20Points))
  (bind PointToAttributeSet_1 (Point%20To%20Attribute%20Set))
  (bind GetPropertyFromObjectPath_1 (Get%20Property%20From%20Object%20Path :bSanitizeOutputAttributeName false :InputSource "Shape Grammar Rules" :OutputAttributeName Rules :PropertyName Rule :StructExtractorBehavior Extract))
  (bind MergeAttributes_0 (Merge%20Attributes))
  (bind CopyAttributes_0 (Copy%20Attributes :bCopyAllAttributes true :InputSource @Last))
  (bind AttributeMathsOp_0 (Attribute%20Maths%20Op :InputSource1 Type :Operation Set))
  (bind AttributeRename_0 (Attribute%20Rename :AttributeToRename "Shape Grammar Rules" :NewAttributeName Rule_Asset))
  ;[0 0 0 0.5] Extract Modules
  (bind MetadataArrayOperation_1 (Metadata%20Array%20Operation :InputSource1 "Modules Infos" :Operation Flatten))
  (bind CopyAttributes_4 (Copy%20Attributes :InputSource size :OutputTarget rulesize))
  (bind ExtractMemberFromStruct_1 (Extract%20Member%20From%20Struct :bExtractAll true :InputSource "Modules Infos"))
  (bind AttributeMathsOp_1 (Attribute%20Maths%20Op :InputSource1 Type :Operation Set))
  (bind AttributeRename_11 (Attribute%20Rename :AttributeToRename size :NewAttributeName module%20size))
  (bind AttributeRename_12 (Attribute%20Rename :AttributeToRename rulesize :NewAttributeName size))
  ;[0 0 0 0.5] Apply non-extracted defaults for modules
  (bind DeleteAttributes_0 (Delete%20Attributes :SelectedAttributes "Slot ID, Rule ID, Type, Rule_Asset, Mesh, Symbol, Scalable, GuessSize, FlipForwardAxis, GuessPivotOffset, PivotOffset, Module Size, MeshForwardAxis, Module Debug Color ")) ; Apply defaults to unset attribs\nfor Modules
  (bind CopyAttributes_2 (Copy%20Attributes :bCopyAllAttributes true :InputSource @Last))
  ;[0 0 0 0.5] Apply non-extracted defaults for slots
  (bind DeleteAttributes_1 (Delete%20Attributes :SelectedAttributes "Shape Grammar Rules"))
  (bind DeleteAttributes_10 (Delete%20Attributes :SelectedAttributes "Slot ID, Symbol, Mirror, Type, Size, Scalable, offset, is corner, ")) ; Apply defaults to unset attribs\nfor slots
  (bind CopyAttributes_7 (Copy%20Attributes :bCopyAllAttributes true :InputSource @Last))
  ;[0 0 0 0.5] Apply non-extracted defaults for rules
  (bind DeleteAttributes_11 (Delete%20Attributes :SelectedAttributes "Slot ID, Rule ID, Type, Rule_Asset, Position On Slot, Grammar, Rand Pos range, Rand Rot range, Rand Scale range, override scale, Pos offset, Override profile grammar, Use up vector, up vector, Spawn Spline Mesh, Spline Mesh angle Threshold, Set Spline Mesh Roll, Spline Roll Degrees")) ; Apply defaults to unset attribs\nfor rules
  (bind DeleteAttributes_12 (Delete%20Attributes :SelectedAttributes "Modules Infos"))
  (bind CopyAttributes_15 (Copy%20Attributes :bCopyAllAttributes true :InputSource @Last))
  ;[0 0 0 0.5] Mergin all types into one table
  (bind MergePoints_1 (Merge%20Points))
  ;[0 0 0 0.5] Renaming extracted attributes to their expected names in operations downstream
  (bind AttributeStringOp_0 (Attribute%20String%20Op :InputSource1 Symbol :Operation Replace))
  (bind AttributeRename_1 (Attribute%20Rename :AttributeToRename DebugColor :NewAttributeName Debug%20Color))
  (bind AttributeRename_2 (Attribute%20Rename :AttributeToRename "Rand Pos Range" :NewAttributeName Rand_Pos_Range))
  (bind AttributeRename_3 (Attribute%20Rename :AttributeToRename "Rand Rot Range" :NewAttributeName Rand_Rot_Range))
  (bind AttributeRename_4 (Attribute%20Rename :AttributeToRename "Rand Scale Range" :NewAttributeName Rand_Scale_Range))
  (bind AttributeRename_5 (Attribute%20Rename :AttributeToRename "Override Scale" :NewAttributeName Override_Scale))
  (bind AttributeRename_6 (Attribute%20Rename :AttributeToRename "Pos Offset" :NewAttributeName Pos_Offset))
  (bind AttributeRename_7 (Attribute%20Rename :AttributeToRename "Override profile grammar" :NewAttributeName Use%20Own%20Grammar))
  (bind AttributeRename_8 (Attribute%20Rename :AttributeToRename Mesh :NewAttributeName Module%20Mesh))
  (bind AttributeRename_13 (Attribute%20Rename :AttributeToRename "Spawn Spline Mesh" :NewAttributeName Spawn%20Spline%20Meshes))
  (bind CopyAttributes_5 (Copy%20Attributes :InputSource "Debug Color" :OutputTarget "Module Debug Color"))
  ;[0 0 0 0.5] To Attribute Set
  (bind PointToAttributeSet_2 (Point%20To%20Attribute%20Set))
  ;[0 0 0 0.5] Commented Out, used to produce the PCG Data Asset used instead of those operations to avoid un-necessary operations.
  ;;[0 0 0 0.5] Initialize Slots, Rule, ModulesInfos defaults
  (bind CreateComplexConstant_3 (Create%20Complex%20Constant :Properties (struct :Value (struct))))
  (bind ExtractMemberFromStruct_4 (Extract%20Member%20From%20Struct :bExtractAll true))
  (bind CreateComplexConstant_4 (Create%20Complex%20Constant :Properties (struct :Value (struct))))
  (bind ExtractMemberFromStruct_5 (Extract%20Member%20From%20Struct :bExtractAll true))
  (bind CreateComplexConstant_5 (Create%20Complex%20Constant :Properties (struct :Value (struct))))
  (bind ExtractMemberFromStruct_6 (Extract%20Member%20From%20Struct :bExtractAll true))
  (bind AttributeSetToPoint_2 (Attribute%20Set%20To%20Point))
  (bind DeleteAttributes_7 (Delete%20Attributes :SelectedAttributes "Shape Grammar Rules"))
  (bind AttributeSetToPoint_3 (Attribute%20Set%20To%20Point))
  (bind DeleteAttributes_8 (Delete%20Attributes :SelectedAttributes "Modules Infos"))
  (bind AttributeSetToPoint_4 (Attribute%20Set%20To%20Point))
  (bind DeleteAttributes_9 (Delete%20Attributes :SelectedAttributes "Size, DebugColor"))
  ;;[0 0 0 0.5] Set defaults that varies for system requirements over struct defaults
  (bind AddAttribute_13 (Add%20Attribute :AttributeTypes (struct :Type Integer32 :Int32Value -1 :DoubleValue -1.0) :OutputTarget "Slot ID"))
  (bind AddAttribute_14 (Add%20Attribute :AttributeTypes (struct :Type SoftObjectPath :Int32Value -1 :DoubleValue -1.0 :SoftObjectPathValue (struct :AssetPath (struct :PackageName /PCGPrimitives/Grammar/DataAsset/RULE_empty :AssetName RULE_empty))) :OutputTarget Rule_Asset))
  (bind AddAttribute_15 (Add%20Attribute :AttributeTypes (struct :Type Integer32 :Int32Value -1) :OutputTarget Type))
  (bind AttributeMathsOp_17 (Attribute%20Maths%20Op :InputSource1 size :Operation Set))
  (bind AddAttribute_19 (Add%20Attribute :AttributeTypes (struct :Type Integer32 :Int32Value -1) :OutputTarget "Rule ID"))
  (bind AttributeMathsOp_20 (Attribute%20Maths%20Op :InputSource1 "position on slot" :Operation Set))
  (bind AttributeMathsOp_21 (Attribute%20Maths%20Op :InputSource1 "Override scale" :Operation Set))
  (bind AttributeMathsOp_22 (Attribute%20Maths%20Op :InputSource1 "Pos Offset" :Operation Set))
  (bind AttributeMathsOp_23 (Attribute%20Maths%20Op :InputSource1 Mesh :Operation Set))
  (bind AddAttribute_20 (Add%20Attribute :AttributeTypes (struct :Type Integer32 :Int32Value -1) :OutputTarget "Module Size"))
  (bind AttributeMathsOp_24 (Attribute%20Maths%20Op :InputSource1 MeshForwardAxis :Operation Set))
  (bind AddAttribute_21 (Add%20Attribute :AttributeTypes (struct :Type Vector4 :Int32Value -1) :OutputTarget "Module Debug Color"))
  (bind AttributeMathsOp_26 (Attribute%20Maths%20Op :InputSource1 Symbol :Operation Set))
  (bind AddAttribute_22 (Add%20Attribute :AttributeTypes (struct :Type Vector4 :Int32Value -1) :OutputTarget DebugColor))
  (bind AttributeMathsOp_2 (Attribute%20Maths%20Op :InputSource1 Symbol :Operation Set))
  ;;[0 0 0 0.5] Combine
  (bind CopyAttributes_13 (Copy%20Attributes :bCopyAllAttributes true :InputSource @Last))
  (bind CopyAttributes_14 (Copy%20Attributes :bCopyAllAttributes true :InputSource @Last))
  ;[1 0 0 0.5] Load pre-saved asset made from the commented out graph
  (bind PCGLoadDataAsset_1 (Load%20PCG%20Data%20Asset :Asset /PCGPrimitives/Grammar/PCG/SGD_DefaultAttributesInit.SGD_DefaultAttributesInit :bTagOutputsBasedOnOutputPins false :Pins [(struct :Label Out :AllowedTypes (struct :Ids [(struct :Struct /Script/PCG.PCGDataTypeInfoPoint)]))]))
  (reroute DefaultAttribs PCGLoadDataAsset_1)
  (connect CopyAttributes_1 Out AddAttribute_1 In)
  (connect DeleteAttributes_7 Out AddAttribute_13 In)
  (connect AddAttribute_13 Out AddAttribute_14 In)
  (connect AddAttribute_14 Out AddAttribute_15 In)
  (connect DeleteAttributes_8 Out AddAttribute_19 In)
  (connect AttributeMathsOp_23 Out AddAttribute_20 In)
  (connect AttributeMathsOp_24 Out AddAttribute_21 In)
  (connect AttributeMathsOp_26 Out AddAttribute_22 In)
  (connect CopyAttributes_0 Out AttributeMathsOp_0 InA)
  (connect ExtractMemberFromStruct_1 Out AttributeMathsOp_1 InA)
  (connect AddAttribute_21 Out AttributeMathsOp_2 InA)
  (connect AddAttribute_15 Out AttributeMathsOp_17 InA)
  (connect AddAttribute_19 Out AttributeMathsOp_20 InA)
  (connect AttributeMathsOp_20 Out AttributeMathsOp_21 InA)
  (connect AttributeMathsOp_21 Out AttributeMathsOp_22 InA)
  (connect DeleteAttributes_9 Out AttributeMathsOp_23 InA)
  (connect AddAttribute_20 Out AttributeMathsOp_24 InA)
  (connect AttributeMathsOp_17 Out AttributeMathsOp_26 InA)
  (connect MetadataArrayOperation_0 Out AttributePartition_2 In)
  (connect AttributeMathsOp_0 Out AttributeRename_0 In)
  (connect AttributeStringOp_0 Out AttributeRename_1 In)
  (connect AttributeRename_1 Out AttributeRename_2 In)
  (connect AttributeRename_2 Out AttributeRename_3 In)
  (connect AttributeRename_3 Out AttributeRename_4 In)
  (connect AttributeRename_4 Out AttributeRename_5 In)
  (connect AttributeRename_5 Out AttributeRename_6 In)
  (connect AttributeRename_6 Out AttributeRename_7 In)
  (connect AttributeRename_7 Out AttributeRename_8 In)
  (connect AttributeMathsOp_1 Out AttributeRename_11 In)
  (connect AttributeRename_11 Out AttributeRename_12 In)
  (connect AttributeRename_8 Out AttributeRename_13 In)
  (connect GetPropertyFromObjectPath_0 Out AttributeSetToPoint_1 In)
  (connect ExtractMemberFromStruct_4 Out AttributeSetToPoint_2 In)
  (connect ExtractMemberFromStruct_5 Out AttributeSetToPoint_3 In)
  (connect ExtractMemberFromStruct_6 Out AttributeSetToPoint_4 In)
  (connect MergePoints_1 Out AttributeStringOp_0 In)
  (connect MergeAttributes_0 Out CopyAttributes_0 Source)
  (connect MergePoints_0 Out CopyAttributes_0 Target)
  (connect AttributeSetToPoint_1 Out CopyAttributes_1 Source)
  (connect AttributeSetToPoint_1 Out CopyAttributes_1 Target)
  (connect DeleteAttributes_0 Out CopyAttributes_2 Source)
  (connect AttributeRename_12 Out CopyAttributes_2 Target)
  (connect AttributePartition_2 Out CopyAttributes_3 Source)
  (connect AttributePartition_2 Out CopyAttributes_3 Target)
  (connect MetadataArrayOperation_1 Out CopyAttributes_4 Source)
  (connect MetadataArrayOperation_1 Out CopyAttributes_4 Target)
  (connect AttributeRename_13 Out CopyAttributes_5 Source)
  (connect AttributeRename_13 Out CopyAttributes_5 Target)
  (connect DeleteAttributes_10 Out CopyAttributes_7 Source)
  (connect DeleteAttributes_1 Out CopyAttributes_7 Target)
  (connect AttributeMathsOp_22 Out CopyAttributes_13 Source)
  (connect AddAttribute_22 Out CopyAttributes_13 Target)
  (connect AttributeMathsOp_2 Out CopyAttributes_14 Source)
  (connect CopyAttributes_13 Out CopyAttributes_14 Target)
  (connect DeleteAttributes_11 Out CopyAttributes_15 Source)
  (connect DeleteAttributes_12 Out CopyAttributes_15 Target)
  (connect DefaultAttribs Out DeleteAttributes_0 In)
  (connect AddAttribute_1 Out DeleteAttributes_1 In)
  (connect AttributeSetToPoint_2 Out DeleteAttributes_7 In)
  (connect AttributeSetToPoint_3 Out DeleteAttributes_8 In)
  (connect AttributeSetToPoint_4 Out DeleteAttributes_9 In)
  (connect DefaultAttribs Out DeleteAttributes_10 In)
  (connect DefaultAttribs Out DeleteAttributes_11 In)
  (connect AttributeRename_0 Out DeleteAttributes_12 In)
  (connect CopyAttributes_4 Out ExtractMemberFromStruct_1 In)
  (connect CreateComplexConstant_3 Out ExtractMemberFromStruct_4 In)
  (connect CreateComplexConstant_4 Out ExtractMemberFromStruct_5 In)
  (connect CreateComplexConstant_5 Out ExtractMemberFromStruct_6 In)
  (connect PointToAttributeSet_1 Out GetPropertyFromObjectPath_1 In)
  (connect GetPropertyFromObjectPath_1 Out MergeAttributes_0 In)
  (connect CopyAttributes_3 Out MergePoints_0 In)
  (connect CopyAttributes_7 Out MergePoints_1 In)
  (connect CopyAttributes_15 Out MergePoints_1 In2)
  (connect CopyAttributes_2 Out MergePoints_1 In3)
  (connect AddAttribute_1 Out MetadataArrayOperation_0 In)
  (connect AttributeRename_0 Out MetadataArrayOperation_1 In)
  (connect MergePoints_0 Out PointToAttributeSet_1 In)
  (connect CopyAttributes_5 Out PointToAttributeSet_2 In)
  (output SGD_Data PointToAttributeSet_2))
