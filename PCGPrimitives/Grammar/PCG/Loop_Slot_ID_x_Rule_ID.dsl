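; PCG loop body that runs the ShapeGrammar_Module subgraph for one Slot ID and Rule ID pair, with optional asset spawning.
; Filters splines and grammar modules to that pair, combines FlipForwardAxis with the spline's Mirror through integer math, and deduplicates modules by Symbol.
; The first matching spline supplies renamed override attributes and, via "override profile grammar", selects between rule and main grammar.
; The loop index serves as the subgraph's seed.
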(graph Loop_Slot_ID_x_Rule_ID
  (params (SpawnAssets :type Bool :default true))
  (inputs (spline%20In :type Spline) (Grammar%20Modules :type Point))
  (bind AttributeFilter_0 (Filter%20Attribute%20Elements :bUseConstantThreshold true :Operator Equal :TargetAttribute $Index :ThresholdAttribute $Index))
  (bind FilterDataByAttribute_0 (Filter%20Data%20By%20Attribute :FilterMode FilterByValue :FilterOperator Equal :TargetAttribute "Slot ID" :Threshold (struct :bUseConstantThreshold false :ThresholdAttribute "Slot ID")))
  (bind FilterDataByAttribute_2 (Filter%20Data%20By%20Attribute :FilterMode FilterByValue :FilterOperator Equal :TargetAttribute "Rule ID" :Threshold (struct :bUseConstantThreshold false :ThresholdAttribute "Rule ID")))
  (bind GetLoopIndex_0 (Get%20Loop%20Index))
  (bind AttributeFilter_1 (Filter%20Attribute%20Elements :bUseConstantThreshold true :Operator Equal :TargetAttribute $Index :ThresholdAttribute $Index))
  (bind AttributeFilter_2 (Filter%20Attribute%20Elements :Operator Equal :TargetAttribute "Slot ID" :ThresholdAttribute "Slot ID"))
  (bind AttributeFilter_9 (Filter%20Attribute%20Elements :Operator Equal :TargetAttribute "Rule ID" :ThresholdAttribute "Rule ID"))
  (bind FilterDataByIndex_4 (Filter%20Data%20By%20Index :SelectedIndices "0"))
  (bind AttributeRemoveDuplicates_0 (Attribute%20Remove%20Duplicates :AttributeSelectors [Symbol]))
  (bind MergePoints_0 (Merge%20Points))
  (bind PointToAttributeSet_0 (Point%20To%20Attribute%20Set))
  (bind ToPoint_0 (To%20Point))
  (bind FilterElementsByIndex_0 (Filter%20Elements%20By%20Index :bSelectIndicesByInput false :SelectedIndices "0"))
  (bind PointToAttributeSet_1 (Point%20To%20Attribute%20Set :PropertiesToConvert (map ("Is Corner" IsCorner) (SlotOrientation Orientation) ("use up vector" use_up_vector) ("up vector" up_vector) ("Spawn Spline Meshes" Spawn_Spline_Meshes) ("Spline Mesh Angle Threshold" Spline_Mesh_Angle_Threshold) ("Set Spline Mesh Roll" Set_Spline_Mesh_Roll) ("Spline Roll Degrees" Spline_Roll_Degrees)))) ; Remaping to expected match names
  (bind Subgraph_1 (Subgraph :Graph /PCGPrimitives/Grammar/PCG/ShapeGrammar_Module.ShapeGrammar_Module :Spawn_Assets SpawnAssets))
  ;[0 0 0 0.5] REVERSE MODULE with Direction
  (bind AttributeCast_1 (Attribute%20Cast :InputSource FlipForwardAxis :OutputType Integer32))
  (bind AttributeMathsOp_1 (Attribute%20Maths%20Op :InputSource1 FlipForwardAxis :Operation Multiply :OutputTarget FlipForwardAxis))
  (bind AttributeMathsOp_4 (Attribute%20Maths%20Op :InputSource1 FlipForwardAxis :OutputTarget FlipForwardAxis))
  (bind ExtractAttribute_8 (Extract%20Attribute%20at%20Index :bForceOutputAttributeToBeInElementsDomain true :InputSource Mirror))
  (bind AttributeMathsOp_12 (Attribute%20Maths%20Op :Operation Multiply))
  (bind AttributeMathsOp_10 (Attribute%20Maths%20Op))
  (bind AttributeMathsOp_5 (Attribute%20Maths%20Op :Operation Multiply))
  (bind AttributeMathsOp_7 (Attribute%20Maths%20Op :InputSource1 FlipForwardAxis :OutputTarget FlipForwardAxis))
  (bind AttributeMathsOp_8 (Attribute%20Maths%20Op :InputSource1 FlipForwardAxis :Operation Multiply :OutputTarget FlipForwardAxis))
  (bind AttributeCast_3 (Attribute%20Cast :InputSource FlipForwardAxis :OutputType Boolean))
  ;[0 0 0 0.5] USE Rule Grammar or Main Grammar
  (bind ExtractAttribute_7 (Extract%20Attribute%20at%20Index :bForceOutputAttributeToBeInElementsDomain true :InputSource grammar :OutputAttributeName grammar))
  (bind ExtractAttribute_16 (Extract%20Attribute%20at%20Index :bForceOutputAttributeToBeInElementsDomain true :InputSource "Rule Grammar" :OutputAttributeName Grammar))
  (bind ExtractAttribute_14 (Extract%20Attribute%20at%20Index :bForceOutputAttributeToBeInElementsDomain true :InputSource "override profile grammar" :OutputAttributeName "override profile grammar"))
  (bind FilterDataByIndex_3 (Filter%20Data%20By%20Index :SelectedIndices "0"))
  (bind Select_2 (Select))
  (reroute First%20Spline FilterDataByIndex_4.InsideFilter)
  (connect AttributeFilter_9 InsideFilter AttributeCast_1 In)
  (connect AttributeMathsOp_8 Out AttributeCast_3 In)
  (connect DefaultInputNode In AttributeFilter_0 In)
  (connect DefaultInputNode In AttributeFilter_1 In)
  (connect AttributeFilter_1 InsideFilter AttributeFilter_2 Filter)
  (connect DefaultInputNode Grammar%20Modules AttributeFilter_2 In)
  (connect AttributeFilter_1 InsideFilter AttributeFilter_9 Filter)
  (connect AttributeFilter_2 InsideFilter AttributeFilter_9 In)
  (connect AttributeCast_1 Out AttributeMathsOp_1 InA)
  (connect AttributeMathsOp_1 Out AttributeMathsOp_4 InA)
  (connect AttributeMathsOp_4 Out AttributeMathsOp_5 InA)
  (connect AttributeMathsOp_10 Out AttributeMathsOp_5 InB)
  (connect AttributeMathsOp_5 Out AttributeMathsOp_7 InA)
  (connect AttributeMathsOp_7 Out AttributeMathsOp_8 InA)
  (connect AttributeMathsOp_12 Out AttributeMathsOp_10 InA)
  (connect ExtractAttribute_8 Out AttributeMathsOp_12 InA)
  (connect AttributeCast_3 Out AttributeRemoveDuplicates_0 In)
  (connect First%20Spline Out ExtractAttribute_7 In)
  (connect FilterDataByIndex_4 InsideFilter ExtractAttribute_8 In)
  (connect First%20Spline Out ExtractAttribute_14 In)
  (connect First%20Spline Out ExtractAttribute_16 In)
  (connect AttributeFilter_0 InsideFilter FilterDataByAttribute_0 Filter)
  (connect DefaultInputNode spline%20In FilterDataByAttribute_0 In)
  (connect AttributeFilter_0 InsideFilter FilterDataByAttribute_2 Filter)
  (connect FilterDataByAttribute_0 InsideFilter FilterDataByAttribute_2 In)
  (connect ExtractAttribute_14 Out FilterDataByIndex_3 In)
  (connect FilterDataByAttribute_2 InsideFilter FilterDataByIndex_4 In)
  (connect ToPoint_0 Out FilterElementsByIndex_0 In)
  (connect AttributeRemoveDuplicates_0 Out MergePoints_0 In)
  (connect MergePoints_0 Out PointToAttributeSet_0 In)
  (connect FilterElementsByIndex_0 In%20Filter PointToAttributeSet_1 In)
  (connect FilterDataByIndex_3 InsideFilter Select_2 bUseInputB)
  (connect ExtractAttribute_16 Out Select_2 Input%20A)
  (connect ExtractAttribute_7 Out Select_2 Input%20B)
  (connect Select_2 Out Subgraph_1 Grammar)
  (connect PointToAttributeSet_0 Out Subgraph_1 modules%20infos)
  (connect PointToAttributeSet_1 Out Subgraph_1 Overrides)
  (connect GetLoopIndex_0 Out Subgraph_1 Seed)
  (connect FilterDataByAttribute_2 InsideFilter Subgraph_1 Splines)
  (connect First%20Spline Out ToPoint_0 In)
  (output Out Subgraph_1.output))
