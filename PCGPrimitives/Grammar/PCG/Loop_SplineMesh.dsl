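; PCG loop body that spawns modules along a spline as static meshes, or as spline meshes where they bend.
; Dot products of forward and up vectors against both wrapped neighbors flag modules below an angle threshold.
; Midpoints between adjacent module ends form a spline, open or closed, resampled for tangents and split into per-segment splines.
; Flagged segments get mesh and optional roll attributes for a nested spawning loop; the rest use Static Mesh Spawner.
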(graph Loop_SplineMesh
  (params (Spline_Mesh_Angle_Threshold :type Double) (Spawn_Spline_Meshes :type Bool) (Set_Spline_Mesh_Roll :type Bool) (up_vector :type Vector) (Spline_Roll_Degrees :type Double))
  (bind StaticMeshSpawner_1 (Static%20Mesh%20Spawner :MeshSelectorParameters (class:/Script/PCG.PCGMeshSelectorByAttribute :AttributeName Mesh :TemplateDescriptor (struct :bHasPerInstanceHitProxies true)) :MeshSelectorType /Script/PCG.PCGMeshSelectorByAttribute))
  (bind ExtractAttribute_1 (Extract%20Attribute%20at%20Index :bForceOutputAttributeToBeInElementsDomain true :InputSource @data.Isclosed))
  ;[0 0 0 0.5 32] Spline Mesh Spawning
  (bind AttributeRename_4 (Attribute%20Rename :AttributeToRename MeshForwardAxis :NewAttributeName ForwardAxis))
  (bind AttributeRename_5 (Attribute%20Rename :AttributeToRename FlipForwardAxis :NewAttributeName ReverseOrientation))
  (bind AttributeRename_6 (Attribute%20Rename :AttributeToRename Mesh :NewAttributeName StaticMesh))
  ;;[0 0 0 0.5] Prepare Spline Mesh Data: Create Points at the Start / End of each Module, on the spline
  (bind ResetPointCenter_3 (Reset%20Point%20Center :PointCenterLocation (0.0 0.5 0.5)))
  (bind AttributeFilter_9 (Filter%20Attribute%20Elements :AttributeTypes (struct :Type Integer32) :bUseConstantThreshold true :Operator Equal :TargetAttribute $Index :ThresholdAttribute @Last))
  (bind GetElementsCount_3 (Get%20Element%20Count))
  (bind AttributeMathsOp_7 (Attribute%20Maths%20Op :Operation Subtract))
  (bind ResetPointCenter_4 (Reset%20Point%20Center :PointCenterLocation (1.0 0.5 0.5)))
  (bind AttributeFilter_10 (Filter%20Attribute%20Elements :AttributeTypes (struct :Type Integer32) :Operator Lesser :TargetAttribute $Index :ThresholdAttribute @Last))
  (bind CreateConstant_1 (Create%20Constant :AttributeTypes (struct :DoubleValue 0.5)))
  (bind AttributeTransformOp_1 (Attribute%20Transform%20Op :InputSource1 $Transform :InputSource2 $Transform :Operation Lerp))
  (bind Branch_5 (Branch))
  (bind MergePoints_5 (Merge%20Points))
  ;;;[0 0 0 0.5] Create Spline From Points and Sample it to get tangent attributes
  (bind CreateSpline_3 (Create%20Spline :ArriveTangentAttribute None :InterpTypeAttribute None :LeaveTangentAttribute None))
  (bind SplineSampler_1 (Spline%20Sampler :SamplerParams (struct :SubdivisionsPerSegment 0 :bComputeTangents true :bUnbounded true :PointSteepness 1.0)))
  ;;[0 0 0 0.5] Create segments points pairs
  (bind AttributeFilter_11 (Filter%20Attribute%20Elements :AttributeTypes (struct :Type Integer32) :bUseConstantThreshold true :TargetAttribute $Index :ThresholdAttribute @Last))
  (bind MergePoints_6 (Merge%20Points))
  (bind GetElementsCount_4 (Get%20Element%20Count))
  (bind AttributeMathsOp_8 (Attribute%20Maths%20Op :Operation Subtract))
  (bind CopyAttributes_9 (Copy%20Attributes :InputSource $Index :OutputTarget SegmentID))
  (bind CopyAttributes_10 (Copy%20Attributes :InputSource $Index :OutputTarget SegmentID))
  (bind MergePoints_7 (Merge%20Points))
  (bind Branch_6 (Branch))
  (bind AttributeFilter_12 (Filter%20Attribute%20Elements :Operator Lesser :TargetAttribute SegmentID :ThresholdAttribute @Last))
  (bind AttributePartition_2 (Attribute%20Partition :PartitionAttributeSelectors [SegmentID]))
  (bind FilterDataByIndex_1 (Filter%20Data%20By%20Index))
  ;;;[0 0 0 0.5] Keep only segments to be spawned as spline meshes
  (bind GetAttributeFromPointIndex_3 (Get%20Attribute%20From%20Point%20Index :bForceOutputAttributeToBeInElementsDomain true :InputSource SpawnAsSplineMesh))
  (bind MergeAttributes_1 (Merge%20Attributes))
  (bind CopyAttributes_14 (Copy%20Attributes :InputSource $Index :OutputTarget IndicesToKeep))
  (bind AttributeFilter_13 (Filter%20Attribute%20Elements :AttributeTypes (struct :Type Boolean :BoolValue true) :bGenerateOutputDataEvenIfEmpty false :bUseConstantThreshold true :Operator Equal :TargetAttribute SpawnAsSplineMesh :ThresholdAttribute @Last))
  (bind AttributeReduce_1 (Attribute%20Reduce :InputSource IndicesToKeep :Operation Join :OutputAttributeName IndicesToKeep))
  ;;[0 0 0 0.5] SplineMesh Roll Fixes
  (bind Branch_7 (Branch :bOutputToB-param Set_Spline_Mesh_Roll))
  (bind CopyAttributes_11 (Copy%20Attributes :InputSource @Last :OutputTarget @Data.SplineUpDir :Source up_vector))
  (bind CopyAttributes_12 (Copy%20Attributes :InputSource @Last :OutputTarget @Data.RollDegrees :Source Spline_Roll_Degrees))
  (bind CopyAttributes_13 (Copy%20Attributes :InputSource @Last :OutputTarget @Data.RollDegreesEnd :Source Spline_Roll_Degrees))
  ;;[0.15 0.15 0.15 0.5] fail safe for no indices
  (bind GetElementsCount_1 (Get%20Element%20Count))
  (bind Branch_0 (Branch))
  ;;[0 0 0 0.5] Spawn Spline Meshes
  (bind Loop_1 (Loop :Graph /PCG/EdMode/DrawSpline/Subgraphs/LinearGrammar_SpawnSplineMeshLoop.LinearGrammar_SpawnSplineMeshLoop :SetSplineMeshRoll Set_Spline_Mesh_Roll))
  ;[0 0 0 0.5] Copy StaticMesh Attrib. from Segment to each Spline
  (bind CreateSpline_4 (Create%20Spline :bApplyCustomTangents true :InterpTypeAttribute None))
  (bind GetAttributeFromPointIndex_2 (Get%20Attribute%20From%20Point%20Index :bForceOutputAttributeToBeInElementsDomain true :InputSource Mesh))
  (bind CopyAttributes_8 (Copy%20Attributes :InputSource Mesh :OutputTarget @Data.StaticMesh))
  ;[0 0 0 0.5 32] Static Mesh Spawning
  ;;[0 0 0 0.5] Filter points for Spline Mesh Spawning based on a defined angle tolerance threshold.
  (bind AttributeFilter_14 (Filter%20Attribute%20Elements :AttributeTypes (struct :Type Integer32) :bUseConstantThreshold true :TargetAttribute $Index :ThresholdAttribute @Last))
  (bind MergePoints_8 (Merge%20Points))
  (bind AttributeVectorOp_4 (Attribute%20Vector%20Op :InputSource1 $Rotation.Forward :InputSource2 $Rotation.Forward :Operation Dot :OutputTarget dot))
  (bind GetElementsCount_5 (Get%20Element%20Count))
  (bind AttributeMathsOp_9 (Attribute%20Maths%20Op :Operation Subtract))
  (bind AttributeFilter_15 (Filter%20Attribute%20Elements :AttributeTypes (struct :Type Integer32) :Operator Equal :TargetAttribute $Index :ThresholdAttribute @Last))
  (bind MergePoints_9 (Merge%20Points))
  (bind AttributeVectorOp_5 (Attribute%20Vector%20Op :InputSource1 $Rotation.Forward :InputSource2 $Rotation.Forward :Operation Dot :OutputTarget dot))
  (bind AttributeMathsOp_10 (Attribute%20Maths%20Op :Operation Min))
  (bind AttributeVectorOp_6 (Attribute%20Vector%20Op :InputSource1 $Rotation.Up :InputSource2 $Rotation.Up :Operation Dot :OutputTarget dot))
  (bind AttributeVectorOp_7 (Attribute%20Vector%20Op :InputSource1 $Rotation.Up :InputSource2 $Rotation.Up :Operation Dot :OutputTarget dot))
  (bind AttributeMathsOp_11 (Attribute%20Maths%20Op :Operation Min))
  (bind AttributeMathsOp_12 (Attribute%20Maths%20Op :Operation Min))
  (bind AttributeCompareOp_1 (Attribute%20Compare%20Op :InputSource1 dot :Operation Less :OutputTarget SpawnAsSplineMesh :InB Spline_Mesh_Angle_Threshold))
  (bind AttributeFilter_17 (Filter%20Attribute%20Elements :AttributeTypes (struct :Type Boolean :DoubleValue 0.995) :bUseConstantThreshold true :Operator Equal :TargetAttribute SpawnAsSplineMesh :ThresholdAttribute @Last))
  (reroute IsClosed_Data_1 ExtractAttribute_1)
  (reroute PointsForSplineMeshSpawning_1 AttributeCompareOp_1)
  (reroute PointsFromModulesStartEnds_1 SplineSampler_1)
  (reroute SegmentsPointPairs_1 FilterDataByIndex_1.InsideFilter)
  (connect AttributeMathsOp_12 Out AttributeCompareOp_1 InA)
  (connect ResetPointCenter_3 Out AttributeFilter_9 In)
  (connect AttributeMathsOp_7 Out AttributeFilter_10 Filter)
  (connect ResetPointCenter_4 Out AttributeFilter_10 In)
  (connect PointsFromModulesStartEnds_1 Out AttributeFilter_11 In)
  (connect AttributeMathsOp_8 Out AttributeFilter_12 Filter)
  (connect Branch_6 Output%20A AttributeFilter_12 In)
  (connect CopyAttributes_14 Out AttributeFilter_13 In)
  (connect DefaultInputNode In AttributeFilter_14 In)
  (connect AttributeMathsOp_9 Out AttributeFilter_15 Filter)
  (connect DefaultInputNode In AttributeFilter_15 In)
  (connect AttributeCompareOp_1 Out AttributeFilter_17 In)
  (connect GetElementsCount_3 Out AttributeMathsOp_7 InA)
  (connect GetElementsCount_4 Out AttributeMathsOp_8 InA)
  (connect GetElementsCount_5 Out AttributeMathsOp_9 InA)
  (connect AttributeVectorOp_4 Out AttributeMathsOp_10 InA)
  (connect AttributeVectorOp_5 Out AttributeMathsOp_10 InB)
  (connect AttributeVectorOp_6 Out AttributeMathsOp_11 InA)
  (connect AttributeVectorOp_7 Out AttributeMathsOp_11 InB)
  (connect AttributeMathsOp_11 Out AttributeMathsOp_12 InA)
  (connect AttributeMathsOp_10 Out AttributeMathsOp_12 InB)
  (connect AttributeFilter_12 InsideFilter AttributePartition_2 In)
  (connect Branch_6 Output%20B AttributePartition_2 In)
  (connect AttributeFilter_13 InsideFilter AttributeReduce_1 In)
  (connect CopyAttributes_8 Out AttributeRename_4 In)
  (connect AttributeRename_4 Out AttributeRename_5 In)
  (connect AttributeRename_5 Out AttributeRename_6 In)
  (connect AttributeFilter_9 OutsideFilter AttributeTransformOp_1 InA)
  (connect AttributeFilter_10 InsideFilter AttributeTransformOp_1 InB)
  (connect CreateConstant_1 Out AttributeTransformOp_1 Ratio)
  (connect DefaultInputNode In AttributeVectorOp_4 InA)
  (connect MergePoints_8 Out AttributeVectorOp_4 InB)
  (connect DefaultInputNode In AttributeVectorOp_5 InA)
  (connect MergePoints_9 Out AttributeVectorOp_5 InB)
  (connect DefaultInputNode In AttributeVectorOp_6 InA)
  (connect MergePoints_8 Out AttributeVectorOp_6 InB)
  (connect DefaultInputNode In AttributeVectorOp_7 InA)
  (connect MergePoints_9 Out AttributeVectorOp_7 InB)
  (connect GetElementsCount_1 Out Branch_0 bOutputToB)
  (connect AttributePartition_2 Out Branch_0 In)
  (connect IsClosed_Data_1 Out Branch_5 bOutputToB)
  (connect AttributeFilter_10 OutsideFilter Branch_5 In)
  (connect IsClosed_Data_1 Out Branch_6 bOutputToB)
  (connect MergePoints_7 Out Branch_6 In)
  (connect AttributeRename_6 Out Branch_7 In)
  (connect GetAttributeFromPointIndex_2 Attribute CopyAttributes_8 Source)
  (connect CreateSpline_4 Out CopyAttributes_8 Target)
  (connect PointsFromModulesStartEnds_1 Out CopyAttributes_9 Source)
  (connect PointsFromModulesStartEnds_1 Out CopyAttributes_9 Target)
  (connect MergePoints_6 Out CopyAttributes_10 Source)
  (connect MergePoints_6 Out CopyAttributes_10 Target)
  (connect Branch_7 Output%20B CopyAttributes_11 Target)
  (connect CopyAttributes_11 Out CopyAttributes_12 Target)
  (connect CopyAttributes_12 Out CopyAttributes_13 Target)
  (connect MergeAttributes_1 Out CopyAttributes_14 Source)
  (connect MergeAttributes_1 Out CopyAttributes_14 Target)
  (connect IsClosed_Data_1 Out CreateSpline_3 bClosedLoop)
  (connect MergePoints_5 Out CreateSpline_3 In)
  (connect SegmentsPointPairs_1 Out CreateSpline_4 In)
  (connect DefaultInputNode In ExtractAttribute_1 In)
  (connect Branch_0 Output%20B FilterDataByIndex_1 In)
  (connect AttributeReduce_1 Out FilterDataByIndex_1 SelectedIndices)
  (connect SegmentsPointPairs_1 Out GetAttributeFromPointIndex_2 In)
  (connect AttributePartition_2 Out GetAttributeFromPointIndex_3 In)
  (connect AttributeReduce_1 Out GetElementsCount_1 In)
  (connect PointsForSplineMeshSpawning_1 Out GetElementsCount_3 In)
  (connect PointsFromModulesStartEnds_1 Out GetElementsCount_4 In)
  (connect DefaultInputNode In GetElementsCount_5 In)
  (connect Branch_7 Output%20A Loop_1 In)
  (connect CopyAttributes_13 Out Loop_1 In)
  (connect GetAttributeFromPointIndex_3 Attribute MergeAttributes_1 In)
  (connect AttributeFilter_9 InsideFilter MergePoints_5 In)
  (connect AttributeTransformOp_1 Out MergePoints_5 In2)
  (connect Branch_5 Output%20A MergePoints_5 In3)
  (connect AttributeFilter_11 InsideFilter MergePoints_6 In)
  (connect AttributeFilter_11 OutsideFilter MergePoints_6 In2)
  (connect CopyAttributes_9 Out MergePoints_7 In)
  (connect CopyAttributes_10 Out MergePoints_7 In2)
  (connect AttributeFilter_14 InsideFilter MergePoints_8 In)
  (connect AttributeFilter_14 OutsideFilter MergePoints_8 In2)
  (connect AttributeFilter_15 InsideFilter MergePoints_9 In)
  (connect AttributeFilter_15 OutsideFilter MergePoints_9 In2)
  (connect PointsForSplineMeshSpawning_1 Out ResetPointCenter_3 In)
  (connect PointsForSplineMeshSpawning_1 Out ResetPointCenter_4 In)
  (connect CreateSpline_3 Out SplineSampler_1 Spline)
  (connect AttributeFilter_17 InsideFilter StaticMeshSpawner_1 In))
