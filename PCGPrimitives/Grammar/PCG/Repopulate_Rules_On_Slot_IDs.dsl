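; PCG graph that repopulates grammar rules onto slot points by matching their Slot ID attribute.
; Rules pass through a default-setting attribute maths op, then an equality filter keeps those whose Slot ID matches the Slots input.
; Copy Points replicates the matching rules onto the slot points, and the output merges those copies with the original slots.
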(graph Repopulate_Rules_On_Slot_IDs
  (inputs (Slots :type Point) (Rules :type Point))
  (bind AttributeMathsOp_0 (Attribute%20Maths%20Op))
  (bind AttributeFilter_0 (Filter%20Attribute%20Elements :Operator Equal :TargetAttribute "Slot ID" :ThresholdAttribute "Slot ID")) ; Match Rules <-> Slots by Slot ID
  (bind CopyPoints_0 (Copy%20Points)) ; Replicate matching Rules onto every Slot point\n(Slots = targets)
  (bind MergePoints_0 (Merge%20Points))
  (connect DefaultInputNode Slots AttributeFilter_0 Filter)
  (connect AttributeMathsOp_0 Out AttributeFilter_0 In)
  (connect DefaultInputNode Rules AttributeMathsOp_0 InA)
  (connect AttributeFilter_0 InsideFilter CopyPoints_0 Source)
  (connect DefaultInputNode Slots CopyPoints_0 Target)
  (connect CopyPoints_0 Out MergePoints_0 In)
  (connect DefaultInputNode Slots MergePoints_0 In)
  (output Out MergePoints_0))
