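; PCG shape-grammar graph that fills splines with mesh modules, outputs the placed points and can spawn static or spline meshes.
; Module infos supply size, pivot and forward axis, optionally guessed from mesh bounds, for Subdivide Spline.
; Corner mode subdivides proxy splines encoding each control point's angle as length, then turns positions into yaw at that point.
; Placed modules get axis, pivot, slot-scale, up-vector and random transform adjustments.
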(graph ShapeGrammar_Module
  (params (Grammar :type String :default "[Small1]{Large3,Large1}*[Large2]") (DebugSegments :type Bool) (ModuleDefinitions :type (Array (Struct /PCGPrimitives/Grammar/Structure/STC_Module.STC_Module)) :default [(struct :Mesh /PCG/GraphTemplates/Meshes/Fence01.Fence01 :Symbol "Large1" :DebugColor (struct :R 255 :A 255)) (struct :Mesh /PCG/GraphTemplates/Meshes/Fence02.Fence02 :Symbol "Large2" :DebugColor (struct :G 255 :A 255)) (struct :Mesh /PCG/GraphTemplates/Meshes/Fence03.Fence03 :Symbol "Small1" :DebugColor (struct :B 255 :A 255)) (struct :Mesh /PCG/GraphTemplates/Meshes/Fence04.Fence04 :Symbol "Large3" :DebugColor (struct :B 255 :R 255 :A 255))]) (ExternalModuleDefinitions :type (Object /Script/Engine.DataTable)) (Seed :type Int :default 5166) (Random_Position_Range :type Vector) (Random_Rotation_Range :type Vector) (Random_Scale_Range :type Vector) (Pos_offset :type Vector) (Scale_override :type Vector :default (1.0 1.0 1.0)) (Orientation :type Bool) (IsCorner :type Bool) (Spawn_Assets :type Bool) (Use_up_vector :type Bool) (up_vector :type Vector) (Spawn_Spline_Meshes :type Bool) (Spline_Mesh_Angle_Threshold :type Double) (Set_Spline_Mesh_Roll :type Bool) (Spline_Roll_Degrees :type Double))
  (inputs (Splines :type Spline) (modules%20infos :type Param))
  (outputs (output :type Point))
  (bind Branch_1 (Branch :bOutputToB-param IsCorner))
  (bind AttributeSetToPoint_1 (Attribute%20Set%20To%20Point))
  (bind CopyAttributes_0 (Copy%20Attributes :InputSource "Module Mesh" :OutputTarget Mesh))
  (bind CopyAttributes_1 (Copy%20Attributes :InputSource "Module Size" :OutputTarget Size))
  (bind BoundsFromMesh_0 (Bounds%20From%20Mesh :MeshAttribute Mesh))
  (bind CopyAttributes_3 (Copy%20Attributes :InputSource $BoundsMin :OutputTarget BoundsMinFromMesh))
  (bind CopyAttributes_4 (Copy%20Attributes :InputSource $BoundsMax :OutputTarget BoundsMaxFromMesh))
  (bind AttributeMathsOp_20 (Attribute%20Maths%20Op :InputSource1 Size :InputSource2 MeshModuleAxisVector :Operation Multiply :OutputTarget Size)) ; Get Size using selected Axis
  (bind AttributeVectorOp_1 (Attribute%20Vector%20Op :InputSource1 Size :Operation Length :OutputTarget Size))
  (bind DeleteAttributes_0 (Delete%20Attributes :SelectedAttributes "MeshModuleAxisVector,GuessSize,GuessPivotOffset"))
  (bind AttributeRename_0 (Attribute%20Rename :AttributeToRename "gram Module debug color" :NewAttributeName DEbugColor))
  (bind AttributeMathsOp_1 (Attribute%20Maths%20Op :InputSource1 DebugColor :Operation Divide))
  (bind PointToAttributeSet_2 (Point%20To%20Attribute%20Set))
  (bind SubdivideSpline_1 (Subdivide%20Spline :bAcceptIncompleteSubdivision true :bForwardAttributesFromModulesInfo true :bModuleInfoAsInput true :bOutputSizeAttribute false :GrammarSelection (struct :GrammarString "[C]{W,W1}*[C]") :ModuleHeight 0.0 :ModuleHeightAttribute ExtentsZ :ModulesInfoAttributeNames (struct :bProvideScalable true) :GrammarString Grammar :Seed-param Seed)) ; Module info is \nembedded in the node
  (bind PointToAttributeSet_0 (Point%20To%20Attribute%20Set))
  (bind SubdivideSpline_0 (Subdivide%20Spline :bAcceptIncompleteSubdivision true :bForwardAttributesFromModulesInfo true :bModuleInfoAsInput true :bOutputSizeAttribute false :GrammarSelection (struct :GrammarString "[C]{W,W1}*[C]") :ModuleHeight 0.0 :ModuleHeightAttribute ExtentsZ :ModulesInfoAttributeNames (struct :bProvideScalable true) :GrammarString Grammar :Seed-param Seed)) ; Module info is \nembedded in the node
  (bind CopyAttributes_20 (Copy%20Attributes :InputSource @Last :OutputTarget @data.IsClosed))
  (bind CopyAttributes_21 (Copy%20Attributes :InputSource @Last :OutputTarget @data.IsClosed))
  (bind CopyAttributes_16 (Copy%20Attributes :InputSource storedPosition :OutputTarget storedPosition))
  (bind CopyAttributes_25 (Copy%20Attributes :InputSource StoredRotation :OutputTarget StoredRotation))
  (bind CopyAttributes_26 (Copy%20Attributes :InputSource AngleToNextPoint :OutputTarget AngleToNextPoint))
  (bind MergePoints_2 (Merge%20Points))
  (bind AttributePartition_0 (Attribute%20Partition :PartitionAttributeSelectors [dataIndex]))
  (bind AttributeMathsOp_0 (Attribute%20Maths%20Op :Operation Multiply :InA Random_Position_Range))
  (bind AttributeMathsOp_5 (Attribute%20Maths%20Op :InA Random_Scale_Range))
  (bind CreateConstant_1 (Create%20Constant :AttributeTypes (struct :Type Vector :DoubleValue 1.0 :VectorValue (1.0 1.0 1.0))))
  (bind AttributeMathsOp_6 (Attribute%20Maths%20Op :Operation Subtract :InB Random_Scale_Range))
  (bind BreakVectorAttribute_2 (Break%20Vector%20Attribute :In Random_Rotation_Range))
  (bind MakeRotatorAttribute_4 (Make%20Rotator%20Attribute :Operation MakeRotFromAngles))
  (bind AttributeMathsOp_3 (Attribute%20Maths%20Op :Operation Multiply :InA Random_Rotation_Range))
  (bind BreakVectorAttribute_3 (Break%20Vector%20Attribute))
  (bind MakeRotatorAttribute_5 (Make%20Rotator%20Attribute :Operation MakeRotFromAngles))
  (bind Branch_2 (Branch :bOutputToB-param IsCorner))
  (bind Branch_4 (Branch :bOutputToB-param Orientation))
  (bind FilterDataByIndex_0 (Filter%20Data%20By%20Index :SelectedIndices "0"))
  (bind ExtractAttribute_0 (Extract%20Attribute%20at%20Index :bForceOutputAttributeToBeInElementsDomain true :InputSource @data.slotScale))
  (bind AttributeMathsOp_2 (Attribute%20Maths%20Op :InputSource1 $Scale.Z :InputSource2 slotScale :Operation Set))
  (bind RemoveEmptyData_0 (Remove%20Empty%20Data))
  (bind GenerateSeed_0 (Generate%20Seed :OutputTarget $seed)) ; reset seed
  (bind TransformPoints_1 (Transform%20Points :bUniformScale false :OffsetMax-param Pos_offset :OffsetMin-param Pos_offset :ScaleMax-param Scale_override :ScaleMin-param Scale_override))
  (bind TransformPoints_0 (Transform%20Points :OffsetMax (-3e+03 0.0 0.0) :OffsetMin (3e+03 0.0 0.0) :OffsetMax-param Random_Position_Range :Seed-param Seed))
  (bind CopyAttributes_5 (Copy%20Attributes :InputSource BoundsMinFromMesh :OutputTarget $BoundsMin))
  (bind CopyAttributes_6 (Copy%20Attributes :InputSource BoundsMaxFromMesh :OutputTarget $BoundsMax))
  (bind AttributeRename_1 (Attribute%20Rename :AttributeToRename Mesh :NewAttributeName Asset))
  (bind ExtractAttribute_1 (Extract%20Attribute%20at%20Index :bForceOutputAttributeToBeInElementsDomain true :InputSource @data.$isclosed :OutputAttributeName @data.IsClosed))
  ;[0 0 0 0.5] Corner treatment
  (bind GetSplineControlPoints_0 (Get%20Spline%20Control%20Points))
  (bind CopyAttributes_24 (Copy%20Attributes :InputSource AngleToNextPoint :OutputTarget AngleToNextPoint))
  (bind ExtractAttribute_8 (Extract%20Attribute%20at%20Index :bForceOutputAttributeToBeInElementsDomain true :InputSource $Position))
  (bind CopyAttributes_17 (Copy%20Attributes :InputSource Position :OutputTarget StoredPosition))
  (bind CopyAttributes_7 (Copy%20Attributes :InputSource $Rotation :OutputTarget StoredRotation))
  (bind AttributeCast_0 (Attribute%20Cast :InputSource StoredRotation :OutputType Rotator))
  (bind WriteDataIndex_0 (Write%20Data%20Index :bAddAttribute true :bAddTag false :IndexAttribute dataIndex :IndexTag "dataIndex"))
  (bind AttributeFilter_7 (Filter%20Attribute%20Elements :AttributeTypes (struct :Type Integer32) :bGenerateOutputDataEvenIfEmpty false :bUseConstantThreshold true :Operator Equal :TargetAttribute $Index :ThresholdAttribute @Last))
  (bind AttributeMathsOp_8 (Attribute%20Maths%20Op :InputSource1 $Position :Operation Set))
  (bind AttributeMathsOp_11 (Attribute%20Maths%20Op :InputSource1 $Position :Operation Set))
  (bind AttributeMathsOp_10 (Attribute%20Maths%20Op :InputSource1 $Position.y :InputSource2 AngleToNextPoint :Operation Set))
  (bind MergePoints_0 (Merge%20Points))
  (bind AttributePartition_1 (Attribute%20Partition :bDoNotPartition false :PartitionAttributeSelectors [dataIndex]))
  (bind CreateSpline_0 (Create%20Spline :InterpTypeAttribute None))
  ;[0 0 0 0.5] Convert Module Axis Enum to Vector
  (bind AttributeFilter_0 (Filter%20Attribute%20Elements :AttributeTypes (struct :Type Integer32 :Int32Value 1) :bGenerateOutputDataEvenIfEmpty false :bUseConstantThreshold true :Operator LesserOrEqual :TargetAttribute MeshForwardAxis :ThresholdAttribute @Last))
  (bind AddAttribute_0 (Add%20Attribute :AttributeTypes (struct :Type Vector :VectorValue (1.0 0.0 0.0)) :OutputTarget MeshModuleAxisVector))
  (bind AttributeFilter_1 (Filter%20Attribute%20Elements :AttributeTypes (struct :Type Integer32 :Int32Value 2) :bGenerateOutputDataEvenIfEmpty false :bUseConstantThreshold true :Operator Equal :TargetAttribute MeshForwardAxis :ThresholdAttribute @Last))
  (bind AddAttribute_3 (Add%20Attribute :AttributeTypes (struct :Type Vector :VectorValue (0.0 1.0 0.0)) :OutputTarget MeshModuleAxisVector))
  (bind AttributeFilter_2 (Filter%20Attribute%20Elements :AttributeTypes (struct :Type Integer32 :Int32Value 3) :bGenerateOutputDataEvenIfEmpty false :bUseConstantThreshold true :Operator GreaterOrEqual :TargetAttribute MeshForwardAxis :ThresholdAttribute @Last))
  (bind AddAttribute_4 (Add%20Attribute :AttributeTypes (struct :Type Vector :VectorValue (0.0 0.0 1.0)) :OutputTarget MeshModuleAxisVector))
  (bind MergePoints_1 (Merge%20Points))
  ;[0 0 0 0.5] Guess Module Size or use as-is
  (bind AttributeCast_2 (Attribute%20Cast :InputSource GuessSize))
  (bind AttributeMathsOp_16 (Attribute%20Maths%20Op :InputSource1 Size :InputSource2 $LocalSize :InputSource3 GuessSize :Operation Lerp))
  ;[0 0 0 0.5] Guess Pivot Offset or Use as-is
  (bind AttributeCast_1 (Attribute%20Cast :InputSource GuessPivotOffset))
  (bind AttributeMathsOp_13 (Attribute%20Maths%20Op :InputSource1 $LocalCenter :Operation Negate :OutputTarget PivotOffset)) ; -LocalCenter is the PivotOffset
  (bind AttributeMathsOp_4 (Attribute%20Maths%20Op :InputSource1 PivotOffset.Z :Operation Set)) ; Setting pivot offset Z to 0, because\nmost of the time, when we want to \ncenter the pivot, this only on X an Y,\nnot Z.
  (bind AttributeMathsOp_15 (Attribute%20Maths%20Op :InputSource1 PivotOffset :InputSource2 PivotOffset :InputSource3 GuessPivotOffset :Operation Lerp))
  ;[0 0 0 0.5] Debug Segments
  (bind Branch_0 (Branch :bOutputToB-param DebugSegments))
  (bind ExtentsModifier_0 (Extents%20Modifier :Extents (0.0 0.0 2.5e+02) :Mode Add))
  (bind CopyAttributes_2 (Copy%20Attributes :InputSource DebugColor :OutputTarget $Color))
  (bind Debug_1 (Debug))
  (bind ExtentsModifier_1 (Extents%20Modifier :Extents (1.0 2.0 1.0) :Mode Multiply))
  (bind Debug_2 (Debug))
  ;[0 0 0 0.5] Apply Scale / Rotation using Forward Axis / Flip Forward Axis
  (bind AttributeCast_3 (Attribute%20Cast :InputSource FlipForwardAxis :OutputType Integer32))
  (bind AttributeMathsOp_24 (Attribute%20Maths%20Op :InputSource1 FlipForwardAxis :Operation Multiply :OutputTarget FlipForwardAxis))
  (bind AttributeMathsOp_25 (Attribute%20Maths%20Op :InputSource1 FlipForwardAxis :OutputTarget FlipForwardAxis))
  (bind AttributeMathsOp_23 (Attribute%20Maths%20Op :InputSource1 $Rotation.Forward :InputSource2 FlipForwardAxis :Operation Multiply :OutputTarget ForwardDir))
  (bind WriteDataIndex_1 (Write%20Data%20Index :bAddAttribute true :bAddTag false :IndexAttribute dataIndex))
  (bind AttributeFilter_3 (Filter%20Attribute%20Elements :AttributeTypes (struct :Type Integer32 :Int32Value 1) :bGenerateOutputDataEvenIfEmpty false :bUseConstantThreshold true :Operator LesserOrEqual :TargetAttribute MeshForwardAxis :ThresholdAttribute @Last))
  (bind AttributeFilter_4 (Filter%20Attribute%20Elements :AttributeTypes (struct :Type Integer32 :Int32Value 2) :bGenerateOutputDataEvenIfEmpty false :bUseConstantThreshold true :Operator Equal :TargetAttribute MeshForwardAxis :ThresholdAttribute @Last))
  (bind AttributeMathsOp_21 (Attribute%20Maths%20Op :InputSource1 $Scale :InputSource2 $Scale.yxz :Operation Set))
  (bind AttributeFilter_5 (Filter%20Attribute%20Elements :AttributeTypes (struct :Type Integer32 :Int32Value 3) :bGenerateOutputDataEvenIfEmpty false :bUseConstantThreshold true :Operator GreaterOrEqual :TargetAttribute MeshForwardAxis :ThresholdAttribute @Last))
  (bind AttributeMathsOp_22 (Attribute%20Maths%20Op :InputSource1 $Scale :InputSource2 $Scale.zyx :Operation Set))
  (bind MakeRotatorAttribute_0 (Make%20Rotator%20Attribute :InputSource1 ForwardDir :InputSource2 $Rotation.UP :Operation MakeRotFromXZ :OutputTarget $Rotation))
  (bind MakeRotatorAttribute_1 (Make%20Rotator%20Attribute :InputSource1 ForwardDir :InputSource2 $Rotation.Up :Operation MakeRotFromZY :OutputTarget $Rotation))
  (bind MakeRotatorAttribute_2 (Make%20Rotator%20Attribute :InputSource1 ForwardDir :InputSource2 $Rotation.UP :Operation MakeRotFromYZ :OutputTarget $Rotation))
  ;[0 0 0 0.5] Apply Pivot Offset
  (bind AttributeMathsOp_27 (Attribute%20Maths%20Op :InputSource1 PivotOffset :InputSource2 $Scale :Operation Multiply)) ; PivotOffset needs to be \nscaled too
  (bind AttributeVectorOp_2 (Attribute%20Vector%20Op :InputSource1 PivotOffset :InputSource2 $Transform :Operation TransformDirection)) ; And rotated
  (bind AttributeMathsOp_28 (Attribute%20Maths%20Op :InputSource1 $Position :InputSource2 PivotOffset)) ; And added to the position
  ;[0 0 0 0.5] Corner Treatment
  (bind AttributeMathsOp_18 (Attribute%20Maths%20Op :Operation Negate :OutputTarget $Position))
  (bind CopyAttributes_13 (Copy%20Attributes :InputSource $Position.Y :OutputTarget $Rotation.Yaw))
  (bind AttributeMathsOp_30 (Attribute%20Maths%20Op :InputSource1 $Rotation.Yaw :InputSource2 StoredRotation.Yaw))
  (bind AttributeMathsOp_12 (Attribute%20Maths%20Op :InputSource1 $Rotation.Yaw))
  (bind CopyAttributes_19 (Copy%20Attributes :InputSource storedPosition :OutputTarget $Position))
  ;[0 0 0 0.5] temp fix, need to figure case for meshForwardAxis Z
  (bind AttributeFilter_9 (Filter%20Attribute%20Elements :AttributeTypes (struct :Type Integer32 :Int32Value 1) :bUseConstantThreshold true :Operator Equal :TargetAttribute MeshForwardAxis :ThresholdAttribute @Last)) ; if Mesh Forward Axis == X or else
  (bind AttributeMathsOp_7 (Attribute%20Maths%20Op :InputSource1 $Scale.Y :InputSource2 slotScale :Operation Set)) ; set slot scale to Y
  (bind AttributeMathsOp_34 (Attribute%20Maths%20Op :InputSource1 $Scale.X :InputSource2 slotScale :Operation Set)) ; or set slot scale to X
  ;[0 0 0 0.5] UP Vector option
  (bind Branch_5 (Branch :bOutputToB-param Use_up_vector))
  (bind BreakVectorAttribute_0 (Break%20Vector%20Attribute :In up_vector))
  (bind AttributeMathsOp_9 (Attribute%20Maths%20Op :InputSource1 $Rotation.Roll :Operation Multiply))
  (bind AttributeMathsOp_14 (Attribute%20Maths%20Op :InputSource1 $Rotation.Pitch :Operation Multiply))
  (bind AttributeMathsOp_17 (Attribute%20Maths%20Op :InputSource1 $Rotation.yaw :Operation Multiply))
  ;[0 0 0 0.5] Spawning
  (bind Branch_3 (Branch :bOutputToB-param Spawn_Assets))
  (bind Branch_6 (Branch :bOutputToB-param Spawn_Spline_Meshes))
  (bind Loop_0 (Loop :Graph /PCGPrimitives/Grammar/PCG/Loop_SplineMesh.Loop_SplineMesh :Set_Spline_Mesh_Roll Set_Spline_Mesh_Roll :Spawn_Spline_Meshes Spawn_Spline_Meshes :Spline_Mesh_Angle_Threshold Spline_Mesh_Angle_Threshold :Spline_Roll_Degrees Spline_Roll_Degrees :up_vector up_vector))
  (bind MergePoints_3 (Merge%20Points))
  (bind StaticMeshSpawner_0 (Static%20Mesh%20Spawner :MeshSelectorParameters (class:/Script/PCG.PCGMeshSelectorByAttribute :AttributeName Mesh :TemplateDescriptor (struct :bHasPerInstanceHitProxies true)) :MeshSelectorType /Script/PCG.PCGMeshSelectorByAttribute))
  (reroute Input%20Node%20Splines Splines)
  (reroute IS%20closed ExtractAttribute_1)
  (connect AttributeFilter_0 InsideFilter AddAttribute_0 In)
  (connect AttributeFilter_1 InsideFilter AddAttribute_3 In)
  (connect AttributeFilter_2 InsideFilter AddAttribute_4 In)
  (connect CopyAttributes_7 Out AttributeCast_0 In)
  (connect AttributeVectorOp_1 Out AttributeCast_1 In)
  (connect MergePoints_1 Out AttributeCast_2 In)
  (connect CopyAttributes_20 Out AttributeCast_3 In)
  (connect CopyAttributes_26 Out AttributeCast_3 In)
  (connect CopyAttributes_4 Out AttributeFilter_0 In)
  (connect CopyAttributes_4 Out AttributeFilter_1 In)
  (connect CopyAttributes_4 Out AttributeFilter_2 In)
  (connect WriteDataIndex_1 Out AttributeFilter_3 In)
  (connect WriteDataIndex_1 Out AttributeFilter_4 In)
  (connect WriteDataIndex_1 Out AttributeFilter_5 In)
  (connect WriteDataIndex_0 Out AttributeFilter_7 In)
  (connect Branch_4 Output%20B AttributeFilter_9 In)
  (connect AttributeRename_0 Out AttributeMathsOp_1 InA)
  (connect Branch_4 Output%20A AttributeMathsOp_2 InA)
  (connect ExtractAttribute_0 Out AttributeMathsOp_2 InB)
  (connect AttributeMathsOp_13 Out AttributeMathsOp_4 InA)
  (connect CreateConstant_1 Out AttributeMathsOp_6 InA)
  (connect AttributeFilter_9 InsideFilter AttributeMathsOp_7 InA)
  (connect ExtractAttribute_0 Out AttributeMathsOp_7 InB)
  (connect AttributeFilter_7 InsideFilter AttributeMathsOp_8 InA)
  (connect Branch_5 Output%20B AttributeMathsOp_9 InA)
  (connect BreakVectorAttribute_0 X AttributeMathsOp_9 InB)
  (connect AttributeMathsOp_11 Out AttributeMathsOp_10 InA)
  (connect AttributeMathsOp_11 Out AttributeMathsOp_10 InB)
  (connect AttributeFilter_7 OutsideFilter AttributeMathsOp_11 InA)
  (connect AttributeMathsOp_30 Out AttributeMathsOp_12 InA)
  (connect AttributeVectorOp_1 Out AttributeMathsOp_13 In)
  (connect AttributeMathsOp_9 Out AttributeMathsOp_14 InA)
  (connect BreakVectorAttribute_0 Y AttributeMathsOp_14 InB)
  (connect AttributeVectorOp_1 Out AttributeMathsOp_15 InA)
  (connect AttributeMathsOp_4 Out AttributeMathsOp_15 InB)
  (connect AttributeCast_1 Out AttributeMathsOp_15 Ratio)
  (connect MergePoints_1 Out AttributeMathsOp_16 InA)
  (connect MergePoints_1 Out AttributeMathsOp_16 InB)
  (connect AttributeCast_2 Out AttributeMathsOp_16 Ratio)
  (connect AttributeMathsOp_14 Out AttributeMathsOp_17 InA)
  (connect BreakVectorAttribute_0 Z AttributeMathsOp_17 InB)
  (connect Branch_2 Output%20B AttributeMathsOp_18 In)
  (connect AttributeMathsOp_16 Out AttributeMathsOp_20 InA)
  (connect AttributeMathsOp_16 Out AttributeMathsOp_20 InB)
  (connect AttributeFilter_4 InsideFilter AttributeMathsOp_21 InA)
  (connect AttributeFilter_4 InsideFilter AttributeMathsOp_21 InB)
  (connect AttributeFilter_5 InsideFilter AttributeMathsOp_22 InA)
  (connect AttributeFilter_5 InsideFilter AttributeMathsOp_22 InB)
  (connect AttributeMathsOp_25 Out AttributeMathsOp_23 InA)
  (connect AttributeMathsOp_25 Out AttributeMathsOp_23 InB)
  (connect AttributeCast_3 Out AttributeMathsOp_24 InA)
  (connect AttributeMathsOp_24 Out AttributeMathsOp_25 InA)
  (connect AttributePartition_0 Out AttributeMathsOp_27 InA)
  (connect AttributePartition_0 Out AttributeMathsOp_27 InB)
  (connect AttributeVectorOp_2 Out AttributeMathsOp_28 InA)
  (connect AttributeVectorOp_2 Out AttributeMathsOp_28 InB)
  (connect CopyAttributes_13 Out AttributeMathsOp_30 InA)
  (connect CopyAttributes_13 Out AttributeMathsOp_30 InB)
  (connect AttributeFilter_9 OutsideFilter AttributeMathsOp_34 InA)
  (connect ExtractAttribute_0 Out AttributeMathsOp_34 InB)
  (connect MergePoints_2 Out AttributePartition_0 In)
  (connect MergePoints_0 Out AttributePartition_1 In)
  (connect DeleteAttributes_0 Out AttributeRename_0 In)
  (connect CopyAttributes_6 Out AttributeRename_1 In)
  (connect DefaultInputNode modules%20infos AttributeSetToPoint_1 In)
  (connect AttributeMathsOp_20 Out AttributeVectorOp_1 In)
  (connect AttributeMathsOp_27 Out AttributeVectorOp_2 In)
  (connect AttributeMathsOp_27 Out AttributeVectorOp_2 Transform)
  (connect CopyAttributes_1 Out BoundsFromMesh_0 In)
  (connect CopyAttributes_20 Out Branch_0 In)
  (connect CopyAttributes_21 Out Branch_0 In)
  (connect Input%20Node%20Splines Out Branch_1 In)
  (connect AttributeMathsOp_28 Out Branch_2 In)
  (connect TransformPoints_0 Out Branch_3 In)
  (connect Branch_2 Output%20A Branch_4 In)
  (connect GenerateSeed_0 Out Branch_5 In)
  (connect Branch_3 Output%20B Branch_6 In)
  (connect AttributeMathsOp_3 Out BreakVectorAttribute_3 In)
  (connect AttributeSetToPoint_1 Out CopyAttributes_0 Source)
  (connect AttributeSetToPoint_1 Out CopyAttributes_0 Target)
  (connect CopyAttributes_0 Out CopyAttributes_1 Source)
  (connect CopyAttributes_0 Out CopyAttributes_1 Target)
  (connect ExtentsModifier_0 Out CopyAttributes_2 Source)
  (connect ExtentsModifier_0 Out CopyAttributes_2 Target)
  (connect BoundsFromMesh_0 Out CopyAttributes_3 Source)
  (connect BoundsFromMesh_0 Out CopyAttributes_3 Target)
  (connect CopyAttributes_3 Out CopyAttributes_4 Source)
  (connect CopyAttributes_3 Out CopyAttributes_4 Target)
  (connect TransformPoints_0 Out CopyAttributes_5 Source)
  (connect TransformPoints_0 Out CopyAttributes_5 Target)
  (connect CopyAttributes_5 Out CopyAttributes_6 Source)
  (connect CopyAttributes_5 Out CopyAttributes_6 Target)
  (connect CopyAttributes_17 Out CopyAttributes_7 Source)
  (connect CopyAttributes_17 Out CopyAttributes_7 Target)
  (connect AttributeMathsOp_18 Out CopyAttributes_13 Source)
  (connect AttributeMathsOp_18 Out CopyAttributes_13 Target)
  (connect CreateSpline_0 Out CopyAttributes_16 Source)
  (connect CopyAttributes_21 Out CopyAttributes_16 Target)
  (connect ExtractAttribute_8 Out CopyAttributes_17 Source)
  (connect CopyAttributes_24 Out CopyAttributes_17 Target)
  (connect AttributeMathsOp_12 Out CopyAttributes_19 Source)
  (connect AttributeMathsOp_12 Out CopyAttributes_19 Target)
  (connect IS%20closed Out CopyAttributes_20 Source)
  (connect SubdivideSpline_0 Out CopyAttributes_20 Target)
  (connect IS%20closed Out CopyAttributes_21 Source)
  (connect SubdivideSpline_1 Out CopyAttributes_21 Target)
  (connect Branch_1 Output%20B CopyAttributes_24 Source)
  (connect GetSplineControlPoints_0 Out CopyAttributes_24 Target)
  (connect CreateSpline_0 Out CopyAttributes_25 Source)
  (connect CopyAttributes_16 Out CopyAttributes_25 Target)
  (connect Input%20Node%20Splines Out CopyAttributes_26 Source)
  (connect CopyAttributes_25 Out CopyAttributes_26 Target)
  (connect AttributePartition_1 Out CreateSpline_0 In)
  (connect CopyAttributes_2 Out Debug_1 In)
  (connect ExtentsModifier_1 Out Debug_2 In)
  (connect AttributeMathsOp_15 Out DeleteAttributes_0 In)
  (connect Branch_0 Output%20B ExtentsModifier_0 In)
  (connect ExtentsModifier_0 Out ExtentsModifier_1 In)
  (connect FilterDataByIndex_0 InsideFilter ExtractAttribute_0 In)
  (connect Input%20Node%20Splines Out ExtractAttribute_1 In)
  (connect CopyAttributes_24 Out ExtractAttribute_8 In)
  (connect Input%20Node%20Splines Out FilterDataByIndex_0 In)
  (connect AttributeMathsOp_2 Out GenerateSeed_0 In)
  (connect CopyAttributes_19 Out GenerateSeed_0 In)
  (connect RemoveEmptyData_0 Out GenerateSeed_0 In)
  (connect Branch_1 Output%20B GetSplineControlPoints_0 In)
  (connect Branch_6 Output%20B Loop_0 In)
  (connect AttributeFilter_3 InsideFilter MakeRotatorAttribute_0 X)
  (connect AttributeFilter_3 InsideFilter MakeRotatorAttribute_0 Z)
  (connect AttributeMathsOp_22 Out MakeRotatorAttribute_1 Y)
  (connect AttributeMathsOp_22 Out MakeRotatorAttribute_1 Z)
  (connect AttributeMathsOp_21 Out MakeRotatorAttribute_2 Y)
  (connect AttributeMathsOp_21 Out MakeRotatorAttribute_2 Z)
  (connect BreakVectorAttribute_2 Y MakeRotatorAttribute_4 Pitch)
  (connect BreakVectorAttribute_2 X MakeRotatorAttribute_4 Roll)
  (connect BreakVectorAttribute_2 Z MakeRotatorAttribute_4 Yaw)
  (connect BreakVectorAttribute_3 Y MakeRotatorAttribute_5 Pitch)
  (connect BreakVectorAttribute_3 X MakeRotatorAttribute_5 Roll)
  (connect BreakVectorAttribute_3 Z MakeRotatorAttribute_5 Yaw)
  (connect AttributeMathsOp_8 Out MergePoints_0 In)
  (connect AttributeMathsOp_10 Out MergePoints_0 In)
  (connect AddAttribute_0 Out MergePoints_1 In)
  (connect AddAttribute_3 Out MergePoints_1 In2)
  (connect AddAttribute_4 Out MergePoints_1 In3)
  (connect MakeRotatorAttribute_0 Out MergePoints_2 In)
  (connect MakeRotatorAttribute_1 Out MergePoints_2 In)
  (connect MakeRotatorAttribute_2 Out MergePoints_2 In)
  (connect Branch_6 Output%20A MergePoints_3 In)
  (connect AttributeMathsOp_1 Out PointToAttributeSet_0 In)
  (connect AttributeMathsOp_1 Out PointToAttributeSet_2 In)
  (connect AttributeMathsOp_7 Out RemoveEmptyData_0 In)
  (connect AttributeMathsOp_34 Out RemoveEmptyData_0 In)
  (connect MergePoints_3 Out StaticMeshSpawner_0 In)
  (connect Branch_1 Output%20A SubdivideSpline_0 In)
  (connect PointToAttributeSet_0 Out SubdivideSpline_0 ModulesInfo)
  (connect CreateSpline_0 Out SubdivideSpline_1 In)
  (connect PointToAttributeSet_2 Out SubdivideSpline_1 ModulesInfo)
  (connect TransformPoints_1 Out TransformPoints_0 In)
  (connect AttributeMathsOp_0 Out TransformPoints_0 OffsetMin)
  (connect MakeRotatorAttribute_5 Out TransformPoints_0 RotationMax)
  (connect MakeRotatorAttribute_4 Out TransformPoints_0 RotationMin)
  (connect AttributeMathsOp_5 Out TransformPoints_0 ScaleMax)
  (connect AttributeMathsOp_6 Out TransformPoints_0 ScaleMin)
  (connect AttributeMathsOp_17 Out TransformPoints_1 In)
  (connect Branch_5 Output%20A TransformPoints_1 In)
  (connect AttributeCast_0 Out WriteDataIndex_0 In)
  (connect AttributeMathsOp_23 Out WriteDataIndex_1 In)
  (output output AttributeRename_1))
