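; PCG graph that lays out shape grammar slots along a cross-section, then places their rules and grammar modules.
; A Z or Y spline, sized by Size or a huge default, is subdivided by either a grammar built from slot symbols or the asset's Main Grammar.
; Slots get optional pivot and centering shifts, then loop through Repopulate_Rules_On_Slot_IDs.
; Rule points inherit slot transform data, offsets and mirrored positions; grammar module points exit separately.
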(graph Slots_CrossSection
  (params (ShapeGrammarDefinition :type (Object /Script/Engine.PrimaryDataAsset)) (Size :type Double) (Use_attribute_for_size :type String))
  (inputs (SGD_data :type Param))
  (outputs (Grammar_module_data :type Point))
  (bind AttributeFilter_0 (Filter%20Attribute%20Elements :AttributeTypes (struct :Type Integer32) :bUseConstantThreshold true :Operator Equal :TargetAttribute Type :ThresholdAttribute @Last)) ; Filter Slot only
  (bind GetPropertyFromObjectPath_7 (Get%20Property%20From%20Object%20Path :PropertyName Use%20Grammar :In ShapeGrammarDefinition))
  (bind Branch_0 (Branch))
  (bind AttributeCast_1 (Attribute%20Cast :InputSource Mirror :OutputTarget MirrorString :OutputType String))
  (bind AttributeStringOp_1 (Attribute%20String%20Op :InputSource1 symbol))
  (bind AttributeStringOp_2 (Attribute%20String%20Op :InputSource1 symbol :InputSource2 MirrorString))
  (bind Gather_0 (Gather))
  (bind AttributeReduce_2 (Attribute%20Reduce :InputSource symbol :Operation Join :OutputAttributeName grammarString))
  (bind GetPropertyFromObjectPath_5 (Get%20Property%20From%20Object%20Path :PropertyName Use%20Grammar :In ShapeGrammarDefinition))
  (bind GetPropertyFromObjectPath_6 (Get%20Property%20From%20Object%20Path :PropertyName Main%20Grammar :In ShapeGrammarDefinition))
  (bind Select_1 (Select))
  (bind CreatePoints_2 (Create%20Points))
  (bind AttributeMathsOp_13 (Attribute%20Maths%20Op :InputSource1 $Position.Z :InputSource2 Size :Operation Set))
  (bind MergePoints_0 (Merge%20Points))
  (bind CreateSpline_0 (Create%20Spline :InterpTypeAttribute None))
  (bind AttributeMathsOp_5 (Attribute%20Maths%20Op :InputSource1 $Position.Y :InputSource2 Size :Operation Set))
  (bind MergePoints_2 (Merge%20Points))
  (bind CreateSpline_1 (Create%20Spline :InterpTypeAttribute None))
  (bind SubdivideSpline_0 (Subdivide%20Spline :bAcceptIncompleteSubdivision true :bForwardAttributesFromModulesInfo true :bModuleInfoAsInput true :bOutputDebugColorAttribute true :GrammarSelection (struct :GrammarString "9A") :ModuleHeight 0.0 :ModulesInfo [(struct)]))
  (bind AttributeSetToPoint_2 (Attribute%20Set%20To%20Point))
  (bind AttributeFilter_1 (Filter%20Attribute%20Elements :bUseConstantThreshold true :Operator Lesser :TargetAttribute Type :ThresholdAttribute @Last)) ; RULES ONLY
  (bind GetPropertyFromObjectPath_8 (Get%20Property%20From%20Object%20Path :PropertyName center%20base :In ShapeGrammarDefinition))
  (bind Branch_1 (Branch))
  (bind Gather_2 (Gather))
  (bind Branch_2 (Branch))
  (bind AttributeMathsOp_9 (Attribute%20Maths%20Op :InputSource1 Size :Operation Divide :InA Size))
  (bind Select_6 (Select))
  (bind AttributePartition_1 (Attribute%20Partition :bAssignIndexPartition true :bDoNotPartition false :PartitionAttributeSelectors [$Index] :PartitionIndexAttributeName slot%20Number))
  (bind Loop_0 (Loop :Graph /PCGPrimitives/Grammar/PCG/Repopulate_Rules_On_Slot_IDs.Repopulate_Rules_On_Slot_IDs))
  (bind MergePoints_1 (Merge%20Points :bMergeMetadata false :MergeTagsMode None))
  (bind GetBounds_1 (Get%20Bounds))
  (bind AttributeMathsOp_2 (Attribute%20Maths%20Op :InputSource1 boundsMax :InputSource2 boundsMin :Operation Subtract :OutputTarget size))
  (bind AttributeMathsOp_0 (Attribute%20Maths%20Op :InputSource1 size.X :InputSource2 size.Y :Operation Max :OutputTarget maxSize))
  (bind AddAttribute_0 (Add%20Attribute :InputSource maxSize :OutputTarget @data.totaleSize))
  (bind AttributeFilter_2 (Filter%20Attribute%20Elements :bUseConstantThreshold true :Operator Equal :TargetAttribute Type :ThresholdAttribute @Last)) ; SLOT POINTS
  (bind AttributeFilter_3 (Filter%20Attribute%20Elements :AttributeTypes (struct :DoubleValue -1.0) :bUseConstantThreshold true :Operator Equal :TargetAttribute Type :ThresholdAttribute @Last)) ; RULES POINTS
  (bind AttributeMathsOp_12 (Attribute%20Maths%20Op :InputSource1 $Position :InputSource2 offset)) ; APPLY Offset to position
  (bind CopyAttributes_1 (Copy%20Attributes :InputSource $Position :OutputTarget $Position)) ; APPLY Slot Position to \nRule position
  (bind CopyAttributes_3 (Copy%20Attributes :InputSource $Extents :OutputTarget $Extents)) ; APPLY Slot Bounds to\nRule Points
  (bind CopyAttributes_4 (Copy%20Attributes :InputSource Mirror :OutputTarget Mirror)) ; APPLY Slot direction \nto Rules
  (bind CopyAttributes_7 (Copy%20Attributes :InputSource "Is Corner" :OutputTarget "Is Corner")) ; IS Corner to Rules
  (bind CopyAttributes_6 (Copy%20Attributes :InputSource $scale.X :OutputTarget slotScale))
  (bind AttributeMathsOp_19 (Attribute%20Maths%20Op :InputSource1 $Position :InputSource2 offset)) ; APPLY Slot Offset to position
  (bind AttributeFilter_4 (Filter%20Attribute%20Elements :AttributeTypes (struct :DoubleValue -2.0) :bUseConstantThreshold true :Operator Equal :TargetAttribute Type :ThresholdAttribute @Last)) ; GRAMMAR MODULES POINTS
  (bind MergePoints_3 (Merge%20Points))
  (bind MergePoints_4 (Merge%20Points))
  ;[0 0 0 0.5] ORIENTATION
  (bind GetPropertyFromObjectPath_0 (Get%20Property%20From%20Object%20Path :PropertyName orientation :In ShapeGrammarDefinition))
  (bind AttributeRename_1 (Attribute%20Rename :AttributeToRename orientation :NewAttributeName bUseInputB))
  (bind Select_2 (Select))
  ;[0 0 0 0.5] IF NO SCALABLE SLOT or USE grammar, DEFAULT HEIGHT to a very big value
  (bind AttributeReduce_0 (Attribute%20Reduce :InputSource Scalable :Operation Max :OutputAttributeName Scalable))
  (bind GetPropertyFromObjectPath_2 (Get%20Property%20From%20Object%20Path :PropertyName Use%20Grammar :In ShapeGrammarDefinition))
  (bind AttributeBooleanOp_2 (Attribute%20Boolean%20Op :Operation Or))
  (bind CreateConstant_0 (Create%20Constant :AttributeTypes (struct :FloatValue 1e+07 :DoubleValue 1e+05) :OutputTarget Size))
  (bind Select_5 (Select :Input%20B Size))
  ;[0 0 0 0.5] COMPENSATE SIZE
  (bind AddAttribute_1 (Add%20Attribute :AttributeTypes (struct :Type Vector) :OutputTarget PivotOffset))
  (bind AttributeMathsOp_18 (Attribute%20Maths%20Op :InputSource1 Size :Operation Multiply :OutputTarget PivotOffset.X))
  (bind AttributeMathsOp_16 (Attribute%20Maths%20Op :InputSource1 PivotOffset :InputSource2 $Scale :Operation Multiply)) ; PivotOffset needs to be \nscaled too
  (bind AttributeVectorOp_0 (Attribute%20Vector%20Op :InputSource1 PivotOffset :InputSource2 $Transform :Operation TransformDirection)) ; And rotated
  (bind AttributeMathsOp_17 (Attribute%20Maths%20Op :InputSource1 $Position :InputSource2 PivotOffset)) ; And added to the position
  ;[0 0 0 0.5] CENTERED on the SPLINE
  (bind GetPropertyFromObjectPath_9 (Get%20Property%20From%20Object%20Path :PropertyName centered%20spline :In ShapeGrammarDefinition))
  (bind Branch_3 (Branch))
  (bind AttributeMathsOp_4 (Attribute%20Maths%20Op :InputSource1 $Position.Y :InputSource2 Size))
  (bind Gather_3 (Gather))
  ;[0 0 0 0.5] MAX Height or Bounds.
  (bind GetBounds_0 (Get%20Bounds))
  (bind AttributeMathsOp_1 (Attribute%20Maths%20Op :InputSource1 BoundsMAx :InputSource2 BoundsMin :Operation Subtract :OutputTarget SizeVec))
  (bind AttributeMathsOp_6 (Attribute%20Maths%20Op :InputSource1 SizeVec.X :InputSource2 SizeVec.Y :Operation Max :OutputTarget Size))
  (bind AttributeMathsOp_8 (Attribute%20Maths%20Op :Operation Max :InA Size))
  (bind AttributeMathsOp_3 (Attribute%20Maths%20Op :InputSource1 Size :Operation Divide))
  ;[0 0 0 0.5] DEBUG
  (bind AttributeMathsOp_14 (Attribute%20Maths%20Op :InputSource1 debugColor :Operation Divide))
  (bind AttributeMathsOp_7 (Attribute%20Maths%20Op :InputSource1 $Extents.YZ :Operation Set))
  (bind CopyAttributes_2 (Copy%20Attributes :InputSource debugColor :OutputTarget $color))
  (bind AttributeNoise_0 (Attribute%20Noise :InputSource $Color))
  (bind Debug_1 (Debug))
  (bind AttributeMathsOp_15 (Attribute%20Maths%20Op :InputSource1 $Position.Y :Operation Set))
  ;[0 0 0 0.5] DEBUG
  (bind ExtentsModifier_1 (Extents%20Modifier :Extents (0.0 0.0 2e+02) :Mode Add))
  (bind ExtentsModifier_0 (Extents%20Modifier :Extents (1.0 2.0 1.0) :Mode Multiply))
  (bind Debug_0 (Debug))
  ;[0 0 0 0.5 50] POSITION on SLOT setup
  (bind AttributeMathsOp_10 (Attribute%20Maths%20Op :InputSource1 $BoundsMax :InputSource2 $BoundsMin :Operation Subtract :OutputTarget SlotSize)) ; Slot SIZE
  (bind AttributeMathsOp_26 (Attribute%20Maths%20Op :InputSource1 "Position On Slot" :Operation Subtract)) ; Zeroing pos on slot
  (bind AttributeMathsOp_20 (Attribute%20Maths%20Op :InputSource1 SlotSize.X :InputSource2 "Position On Slot" :Operation Multiply :OutputTarget "Position On Slot"))
  (bind AttributeMathsOp_21 (Attribute%20Maths%20Op :InputSource1 "Position On Slot" :InputSource2 Mirror :Operation Multiply :OutputTarget "Position On Slot"))
  (bind AttributeMathsOp_23 (Attribute%20Maths%20Op :InputSource1 "Position On Slot" :InputSource2 $position.Y :OutputTarget $position.Y))
  ;;[0.15 0.15 0.15 0.5] Direction Handling
  (bind AttributeCast_2 (Attribute%20Cast :InputSource Mirror :OutputType Integer32))
  ;;[0.15 0.15 0.15 0.5] Normalizing direction
  (bind AttributeMathsOp_24 (Attribute%20Maths%20Op :InputSource1 Mirror :Operation Multiply))
  (bind AttributeMathsOp_25 (Attribute%20Maths%20Op :InputSource1 Mirror :Operation Subtract))
  (bind AttributeMathsOp_22 (Attribute%20Maths%20Op :InputSource1 Mirror :Operation Multiply))
  ;[0 0 0 0.5] DEBUG
  (bind CopyAttributes_8 (Copy%20Attributes :InputSource $position :OutputTarget position))
  (bind PointToAttributeSet_0 (Point%20To%20Attribute%20Set))
  (bind DeleteAttributes_0 (Delete%20Attributes :Operation KeepSelectedAttributes :SelectedAttributes "position"))
  (bind MergeAttributes_0 (Merge%20Attributes))
  ;[0 0 0 0.5] DEBUG
  (bind CopyAttributes_9 (Copy%20Attributes :InputSource $position :OutputTarget position))
  (bind PointToAttributeSet_1 (Point%20To%20Attribute%20Set))
  (bind DeleteAttributes_1 (Delete%20Attributes :Operation KeepSelectedAttributes :SelectedAttributes "position_on_slot, offset, position, Mirror, slotSize"))
  (bind MergeAttributes_1 (Merge%20Attributes))
  ;[0 0 0 0.5] DEBUG
  (bind PointToAttributeSet_3 (Point%20To%20Attribute%20Set))
  (bind DeleteAttributes_3 (Delete%20Attributes :Operation KeepSelectedAttributes :SelectedAttributes "Mirror"))
  (bind MergeAttributes_2 (Merge%20Attributes))
  (reroute Enforce%20Height AttributeBooleanOp_2)
  (reroute Grammar%20Module%20Data_1 MergePoints_3)
  (reroute NamedRerouteDeclaration_1 SGD_data)
  (connect AttributeMathsOp_0 Out AddAttribute_0 Attributes)
  (connect Loop_0 Out AddAttribute_0 In)
  (connect Branch_1 Output%20A AddAttribute_1 In)
  (connect AttributeReduce_0 Out AttributeBooleanOp_2 InA)
  (connect GetPropertyFromObjectPath_2 Out AttributeBooleanOp_2 InB)
  (connect Branch_0 Output%20A AttributeCast_1 In)
  (connect AttributeMathsOp_26 Out AttributeCast_2 In)
  (connect NamedRerouteDeclaration_1 Out AttributeFilter_0 In)
  (connect AttributeSetToPoint_2 Out AttributeFilter_1 In)
  (connect AddAttribute_0 Out AttributeFilter_2 In)
  (connect AddAttribute_0 Out AttributeFilter_3 In)
  (connect AddAttribute_0 Out AttributeFilter_4 In)
  (connect AttributeMathsOp_2 Out AttributeMathsOp_0 InA)
  (connect AttributeMathsOp_2 Out AttributeMathsOp_0 InB)
  (connect GetBounds_0 Out AttributeMathsOp_1 InA)
  (connect GetBounds_0 Out AttributeMathsOp_1 InB)
  (connect GetBounds_1 Out AttributeMathsOp_2 InA)
  (connect GetBounds_1 Out AttributeMathsOp_2 InB)
  (connect AttributeMathsOp_8 Out AttributeMathsOp_3 InA)
  (connect Branch_3 Output%20B AttributeMathsOp_4 InA)
  (connect Select_6 Out AttributeMathsOp_4 InB)
  (connect CreatePoints_2 Out AttributeMathsOp_5 InA)
  (connect Select_5 Out AttributeMathsOp_5 InB)
  (connect AttributeMathsOp_1 Out AttributeMathsOp_6 InA)
  (connect AttributeMathsOp_1 Out AttributeMathsOp_6 InB)
  (connect AttributeMathsOp_14 Out AttributeMathsOp_7 InA)
  (connect AttributeMathsOp_6 Out AttributeMathsOp_8 InB)
  (connect AttributeMathsOp_19 Out AttributeMathsOp_10 InA)
  (connect AttributeMathsOp_19 Out AttributeMathsOp_10 InB)
  (connect AttributeFilter_2 InsideFilter AttributeMathsOp_12 InA)
  (connect AttributeFilter_2 InsideFilter AttributeMathsOp_12 InB)
  (connect CreatePoints_2 Out AttributeMathsOp_13 InA)
  (connect Select_5 Out AttributeMathsOp_13 InB)
  (connect Gather_3 Out AttributeMathsOp_14 InA)
  (connect AttributeNoise_0 Out AttributeMathsOp_15 InA)
  (connect AttributeMathsOp_18 Out AttributeMathsOp_16 InA)
  (connect AttributeMathsOp_18 Out AttributeMathsOp_16 InB)
  (connect AttributeVectorOp_0 Out AttributeMathsOp_17 InA)
  (connect AttributeVectorOp_0 Out AttributeMathsOp_17 InB)
  (connect AddAttribute_1 Out AttributeMathsOp_18 InA)
  (connect CopyAttributes_6 Out AttributeMathsOp_19 InA)
  (connect AttributeMathsOp_12 Out AttributeMathsOp_19 InB)
  (connect AttributeCast_2 Out AttributeMathsOp_20 InA)
  (connect AttributeCast_2 Out AttributeMathsOp_20 InB)
  (connect AttributeMathsOp_20 Out AttributeMathsOp_21 InA)
  (connect AttributeMathsOp_22 Out AttributeMathsOp_21 InB)
  (connect AttributeMathsOp_25 Out AttributeMathsOp_22 InA)
  (connect AttributeMathsOp_21 Out AttributeMathsOp_23 InA)
  (connect AttributeMathsOp_21 Out AttributeMathsOp_23 InB)
  (connect AttributeCast_2 Out AttributeMathsOp_24 InA)
  (connect AttributeMathsOp_24 Out AttributeMathsOp_25 InA)
  (connect AttributeMathsOp_10 Out AttributeMathsOp_26 InA)
  (connect CopyAttributes_2 Out AttributeNoise_0 In)
  (connect Gather_3 Out AttributePartition_1 In)
  (connect AttributeFilter_0 InsideFilter AttributeReduce_0 In)
  (connect AttributeStringOp_2 Out AttributeReduce_2 In)
  (connect GetPropertyFromObjectPath_0 Out AttributeRename_1 In)
  (connect NamedRerouteDeclaration_1 Out AttributeSetToPoint_2 In)
  (connect AttributeCast_1 Out AttributeStringOp_1 InA)
  (connect AttributeStringOp_1 Out AttributeStringOp_2 InA)
  (connect AttributeStringOp_1 Out AttributeStringOp_2 InB)
  (connect AttributeMathsOp_16 Out AttributeVectorOp_0 In)
  (connect AttributeMathsOp_16 Out AttributeVectorOp_0 Transform)
  (connect GetPropertyFromObjectPath_7 Out Branch_0 bOutputToB)
  (connect AttributeFilter_0 InsideFilter Branch_0 In)
  (connect GetPropertyFromObjectPath_8 Out Branch_1 bOutputToB)
  (connect SubdivideSpline_0 Out Branch_1 In)
  (connect Enforce%20Height Out Branch_2 bOutputToB)
  (connect SubdivideSpline_0 Out Branch_2 In)
  (connect GetPropertyFromObjectPath_9 Out Branch_3 bOutputToB)
  (connect Gather_2 Out Branch_3 In)
  (connect AttributeMathsOp_12 Out CopyAttributes_1 Source)
  (connect AttributeFilter_3 InsideFilter CopyAttributes_1 Target)
  (connect AttributeMathsOp_7 Out CopyAttributes_2 Source)
  (connect AttributeMathsOp_7 Out CopyAttributes_2 Target)
  (connect AttributeMathsOp_12 Out CopyAttributes_3 Source)
  (connect CopyAttributes_1 Out CopyAttributes_3 Target)
  (connect AttributeMathsOp_12 Out CopyAttributes_4 Source)
  (connect CopyAttributes_3 Out CopyAttributes_4 Target)
  (connect AttributeMathsOp_12 Out CopyAttributes_6 Source)
  (connect CopyAttributes_7 Out CopyAttributes_6 Target)
  (connect AttributeMathsOp_12 Out CopyAttributes_7 Source)
  (connect CopyAttributes_4 Out CopyAttributes_7 Target)
  (connect AttributeMathsOp_12 Out CopyAttributes_8 Source)
  (connect AttributeMathsOp_12 Out CopyAttributes_8 Target)
  (connect AttributeMathsOp_23 Out CopyAttributes_9 Source)
  (connect AttributeMathsOp_23 Out CopyAttributes_9 Target)
  (connect MergePoints_0 Out CreateSpline_0 In)
  (connect MergePoints_2 Out CreateSpline_1 In)
  (connect ExtentsModifier_0 Out Debug_0 In)
  (connect AttributeNoise_0 Out Debug_1 In)
  (connect PointToAttributeSet_0 Out DeleteAttributes_0 In)
  (connect PointToAttributeSet_1 Out DeleteAttributes_1 In)
  (connect PointToAttributeSet_3 Out DeleteAttributes_3 In)
  (connect ExtentsModifier_1 Out ExtentsModifier_0 In)
  (connect Gather_3 Out ExtentsModifier_1 In)
  (connect AttributeStringOp_2 Out Gather_0 In)
  (connect Branch_0 Output%20B Gather_0 In)
  (connect AttributeMathsOp_17 Out Gather_2 In)
  (connect Branch_1 Output%20B Gather_2 In2)
  (connect Branch_3 Output%20A Gather_3 In)
  (connect AttributeMathsOp_4 Out Gather_3 In2)
  (connect Branch_2 Output%20A GetBounds_0 In)
  (connect MergePoints_1 Out GetBounds_1 In)
  (connect AttributeFilter_1 InsideFilter Loop_0 Rules)
  (connect AttributePartition_1 Out Loop_0 Slots)
  (connect DeleteAttributes_0 Out MergeAttributes_0 In)
  (connect DeleteAttributes_1 Out MergeAttributes_1 In)
  (connect DeleteAttributes_3 Out MergeAttributes_2 In)
  (connect CreatePoints_2 Out MergePoints_0 In)
  (connect AttributeMathsOp_13 Out MergePoints_0 In2)
  (connect AttributePartition_1 Out MergePoints_1 In)
  (connect CreatePoints_2 Out MergePoints_2 In)
  (connect AttributeMathsOp_5 Out MergePoints_2 In2)
  (connect AttributeFilter_4 InsideFilter MergePoints_3 In)
  (connect AttributeMathsOp_23 Out MergePoints_4 In)
  (connect CopyAttributes_8 Out PointToAttributeSet_0 In)
  (connect CopyAttributes_9 Out PointToAttributeSet_1 In)
  (connect AttributeCast_2 Out PointToAttributeSet_3 In)
  (connect AttributeMathsOp_22 Out PointToAttributeSet_3 In)
  (connect GetPropertyFromObjectPath_5 Out Select_1 bUseInputB)
  (connect AttributeReduce_2 Out Select_1 Input%20A)
  (connect GetPropertyFromObjectPath_6 Out Select_1 Input%20B)
  (connect AttributeRename_1 Out Select_2 bUseInputB)
  (connect CreateSpline_0 Out Select_2 Input%20A)
  (connect CreateSpline_1 Out Select_2 Input%20B)
  (connect AttributeBooleanOp_2 Out Select_5 bUseInputB)
  (connect CreateConstant_0 Out Select_5 Input%20A)
  (connect Enforce%20Height Out Select_6 bUseInputB)
  (connect AttributeMathsOp_3 Out Select_6 Input%20A)
  (connect AttributeMathsOp_9 Out Select_6 Input%20B)
  (connect Select_1 Out SubdivideSpline_0 GrammarString)
  (connect Select_2 Out SubdivideSpline_0 In)
  (connect Gather_0 Out SubdivideSpline_0 ModulesInfo)
  (output Grammar_module_data Grammar%20Module%20Data_1)
  (output Out MergePoints_4))
