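; PCG graph that generates shape-grammar cross-section slots and runs spline creation for each rule point.
; It extracts Size from the input attribute named by UseAttributeForSize and feeds it, with SGD Data and the ShapeGrammarDefinition, into the Slots_CrossSection subgraph.
; The resulting points are partitioned by $Index so Create_Spline_Per_Rule loops once per point against the input splines.
; It outputs the loop result and the subgraph's grammar module data.
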(graph Slots_CrossSection_Loop
  (params (ShapeGrammarDefinition :type (SoftObject /PCGPrimitives/Grammar/BP/BP_ShapeGrammarDefinition.BP_ShapeGrammarDefinition_C)) (UseAttributeForSize :type String))
  (inputs (SGD%20Data :type Param))
  (outputs (Grammar%20Modules :type Point))
  (bind ExtractAttribute_0 (Extract%20Attribute%20at%20Index :bForceOutputAttributeToBeInElementsDomain true :OutputAttributeName Size :InputSource-param UseAttributeForSize)) ; Pull size from attribute named in\n'Use Attribute for Size' (per-element override)
  (bind Subgraph_0 (Subgraph :Graph /PCGPrimitives/Grammar/PCG/Slots_CrossSection.Slots_CrossSection :ShapeGrammarDefinition ShapeGrammarDefinition))
  (bind AttributePartition_1 (Attribute%20Partition :PartitionAttributeSelectors [$Index])) ; Partition by $Index so Create Spline Per Rule\niterates once per rule point
  (bind Loop_1 (Loop :Graph /PCGPrimitives/Grammar/PCG/Create_Spline_Per_Rule.Create_Spline_Per_Rule))
  (connect Subgraph_0 Out AttributePartition_1 In)
  (connect DefaultInputNode In ExtractAttribute_0 In)
  (connect AttributePartition_1 Out Loop_1 Slot%20Points)
  (connect DefaultInputNode In Loop_1 Splines)
  (connect DefaultInputNode SGD%20Data Subgraph_0 SGD_data)
  (connect ExtractAttribute_0 Out Subgraph_0 Size)
  (output Grammar%20Modules Subgraph_0.Grammar_module_data)
  (output Out Loop_1))
