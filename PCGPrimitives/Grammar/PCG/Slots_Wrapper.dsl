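; PCG wrapper graph that applies a ShapeGrammarDefinition to input splines, with optional asset spawning.
; It tags splines with definition properties, optionally reverses them, and feeds a fixed or attribute-driven Size into the Slots_CrossSection subgraph.
; Per-slot points and whole or split splines, chosen by "Spawn on", drive per-rule spline creation and angle calculation.
; Slot ID and Rule ID combine into a Looping_ID whose partitions run Loop_Slot_ID_x_Rule_ID with grammar module data.
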(graph Slots_Wrapper
  (params (Size :type Double :default 1.5e+03) (ShapeGrammarDefinition :type (SoftObject /PCGPrimitives/Grammar/BP/BP_ShapeGrammarDefinition.BP_ShapeGrammarDefinition_C)) (UseAttributeForSize :type String) (SpawnAssets :type Bool :default true))
  (inputs (SGD_Data :type Param))
  (bind WriteDataIndex_0 (Write%20Data%20Index :bAddAttribute true :bAddTag false :IndexAttribute SplineIndex))
  (bind CopyAttributes_4 (Copy%20Attributes :InputSource @data.$isClosed :OutputTarget InputSplineIsClosed))
  (bind RemoveEmptyData_0 (Remove%20Empty%20Data))
  (bind Subgraph_1 (Subgraph :Graph /PCGPrimitives/Primitives/Shared/Splines_CalculateAngleToNextPoint.Splines_CalculateAngleToNextPoint))
  (bind Loop_3 (Loop :Graph /PCGPrimitives/Grammar/PCG/Loop_Slot_ID_x_Rule_ID.Loop_Slot_ID_x_Rule_ID :SpawnAssets SpawnAssets))
  ;[0 0 0 0.5 50] STORING SGD data on the spline to use downstream
  (bind GetPropertyFromObjectPath_3 (Get%20Property%20From%20Object%20Path :PropertyName Rule%20Grammar :In ShapeGrammarDefinition))
  (bind AddAttribute_1 (Add%20Attribute :AttributeTypes (struct :Type Boolean) :OutputTarget "Rule Grammar"))
  (bind GetPropertyFromObjectPath_2 (Get%20Property%20From%20Object%20Path :OutputAttributeName None :PropertyName orientation :In ShapeGrammarDefinition))
  (bind AddAttribute_3 (Add%20Attribute :OutputTarget SlotOrientation))
  (bind GetPropertyFromObjectPath_0 (Get%20Property%20From%20Object%20Path :OutputAttributeName reverse :PropertyName reverse%20spline :In ShapeGrammarDefinition))
  (bind Branch_1 (Branch))
  (bind SplineDirection_1 (Spline%20Direction))
  (bind Gather_0 (Gather))
  ;[0 0 0 0.5] Cross section, supporting size by attribute or fixed size
  (bind Branch_8 (Branch))
  (bind ExtractAttributeAtIndex_1 (Extract%20Attribute%20at%20Index :bForceOutputAttributeToBeInElementsDomain true :OutputAttributeName Size :InputSource-param UseAttributeForSize)) ; Pull size from attribute named in\n'Use Attribute for Size' (per-element override)
  (bind Select_0 (Select :Input%20B Size))
  (bind Subgraph_2 (Subgraph :Graph /PCGPrimitives/Grammar/PCG/Slots_CrossSection.Slots_CrossSection :ShapeGrammarDefinition ShapeGrammarDefinition))
  (bind AttributePartition_0 (Attribute%20Partition :PartitionAttributeSelectors [$Index]))
  (bind GetPropertyFromObjectPath_6 (Get%20Property%20From%20Object%20Path :OutputAttributeName None :PropertyName Spawn%20on :In ShapeGrammarDefinition))
  (bind Branch_7 (Branch))
  (bind CleanSpline_4 (Clean%20Spline))
  (bind SplitSpline_5 (Split%20Spline :Attribute $position.X :bUseConstant false :Mode ByPredicateOnControlPoints))
  (bind Loop_0 (Loop :Graph /PCGPrimitives/Grammar/PCG/Create_Spline_Per_Rule.Create_Spline_Per_Rule))
  ;[0 0 0 0.5] ITERATOR Slot ID x Rule ID
  (bind ExtractAttribute_0 (Extract%20Attribute%20at%20Index :bForceOutputAttributeToBeInElementsDomain true :InputSource "slot ID"))
  (bind MergeAttributes_0 (Merge%20Attributes))
  (bind DeleteTags_0 (Delete%20Tags :Operation KeepOnlySelectedTags))
  (bind ExtractAttribute_1 (Extract%20Attribute%20at%20Index :bForceOutputAttributeToBeInElementsDomain true :InputSource "Rule ID"))
  (bind MergeAttributes_1 (Merge%20Attributes))
  (bind DeleteTags_1 (Delete%20Tags :Operation KeepOnlySelectedTags))
  (bind CopyAttributes_0 (Copy%20Attributes :InputSource @Last))
  (bind AttributeCast_0 (Attribute%20Cast :InputSource "slot ID" :OutputTarget "slot ID string" :OutputType String))
  (bind AttributeCast_2 (Attribute%20Cast :InputSource "Rule ID" :OutputTarget "Rule ID string" :OutputType String))
  (bind AttributeStringOp_0 (Attribute%20String%20Op :InputSource1 "slot ID String" :InputSource2 "Rule ID string" :OutputTarget Looping_ID))
  (bind AttributePartition_2 (Attribute%20Partition :PartitionAttributeSelectors [looping_ID]))
  ;[0 0 0 0.5] Use Attribute Height as Bool
  (bind CreateConstant_2 (Create%20Constant :AttributeTypes (struct :Type String)))
  (bind AttributeCompareOp_2 (Attribute%20Compare%20Op :InA UseAttributeForSize))
  (reroute Grammar%20module%20data Subgraph_2.Grammar_module_data)
  (reroute NamedRerouteDeclaration_0 SGD_Data)
  (reroute UseAttributeForSize%20Bool AttributeCompareOp_2)
  (connect GetPropertyFromObjectPath_3 Out AddAttribute_1 Attributes)
  (connect WriteDataIndex_0 Out AddAttribute_1 In)
  (connect GetPropertyFromObjectPath_2 Out AddAttribute_3 Attributes)
  (connect AddAttribute_1 Out AddAttribute_3 In)
  (connect CopyAttributes_0 Out AttributeCast_0 In)
  (connect AttributeCast_0 Out AttributeCast_2 In)
  (connect CreateConstant_2 Out AttributeCompareOp_2 InB)
  (connect Subgraph_2 Out AttributePartition_0 In)
  (connect AttributeStringOp_0 Out AttributePartition_2 In)
  (connect AttributeCast_2 Out AttributeStringOp_0 InA)
  (connect AttributeCast_2 Out AttributeStringOp_0 InB)
  (connect GetPropertyFromObjectPath_0 Out Branch_1 bOutputToB)
  (connect AddAttribute_3 Out Branch_1 In)
  (connect GetPropertyFromObjectPath_6 Out Branch_7 bOutputToB)
  (connect RemoveEmptyData_0 Out Branch_7 In)
  (connect UseAttributeForSize%20Bool Out Branch_8 bOutputToB)
  (connect RemoveEmptyData_0 Out Branch_8 In)
  (connect Branch_7 Output%20A CleanSpline_4 In)
  (connect DeleteTags_1 Out CopyAttributes_0 Source)
  (connect DeleteTags_0 Out CopyAttributes_0 Target)
  (connect Gather_0 Out CopyAttributes_4 Source)
  (connect Gather_0 Out CopyAttributes_4 Target)
  (connect MergeAttributes_0 Out DeleteTags_0 In)
  (connect MergeAttributes_1 Out DeleteTags_1 In)
  (connect Subgraph_1 Out ExtractAttribute_0 In)
  (connect Subgraph_1 Out ExtractAttribute_1 In)
  (connect Branch_8 Output%20A ExtractAttributeAtIndex_1 In)
  (connect Branch_1 Output%20A Gather_0 In)
  (connect SplineDirection_1 Out Gather_0 In)
  (connect AttributePartition_0 Out Loop_0 Slot%20Points)
  (connect Branch_7 Output%20B Loop_0 Splines)
  (connect SplitSpline_5 Out Loop_0 Splines)
  (connect Grammar%20module%20data Out Loop_3 Grammar%20Modules)
  (connect AttributePartition_2 Out Loop_3 In)
  (connect Subgraph_1 Out Loop_3 spline%20In)
  (connect ExtractAttribute_0 Out MergeAttributes_0 In)
  (connect ExtractAttribute_1 Out MergeAttributes_1 In)
  (connect CopyAttributes_4 Out RemoveEmptyData_0 In)
  (connect UseAttributeForSize%20Bool Out Select_0 bUseInputB)
  (connect ExtractAttributeAtIndex_1 Out Select_0 Input%20A)
  (connect Branch_1 Output%20B SplineDirection_1 In)
  (connect CleanSpline_4 Out SplitSpline_5 In)
  (connect Loop_0 Out Subgraph_1 In)
  (connect NamedRerouteDeclaration_0 Out Subgraph_2 SGD_data)
  (connect Select_0 Out Subgraph_2 Size)
  (connect DefaultInputNode In WriteDataIndex_0 In)
  (output Out Loop_3))
